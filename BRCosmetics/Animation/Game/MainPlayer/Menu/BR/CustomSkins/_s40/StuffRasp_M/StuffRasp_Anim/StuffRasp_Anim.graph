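# Level Sequence that authors a main-menu character animation for the StuffRasp_M custom skin on a FortPlayerMannequin binding.
# A Control Rig track on Fortnite_Biped_CtrlRig drives the pose through keyed body controls, while the skeletal animation track is empty.
# Legs are set to IK with fixed foot targets and animated pole vectors; arms, spine and neck use FK.
# Opening key values recur at the end, suggesting a seamless loop.
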LevelSequence /BRCosmetics/Animation/Game/MainPlayer/Menu/BR/CustomSkins/_s40/StuffRasp_M/StuffRasp_Anim/StuffRasp_Anim
bindings [FortPlayerMannequin {
  id=#1
  possessed_class=FortPlayerMannequin
  tracks [
   {type=MovieSceneSpawnTrack, sections [{type=MovieSceneSpawnSection, channels [None {type=MovieSceneScriptingBoolChannel, default=true}]}]}
   {
    type=MovieScene3DTransformTrack
    sections [{
      type=MovieScene3DTransformSection
      channels [
       Location.X {type=MovieSceneScriptingDoubleChannel, default=0, const [0,1]}
       Location.Y {type=MovieSceneScriptingDoubleChannel, default=0, const [0,1]}
       Location.Z {type=MovieSceneScriptingDoubleChannel, default=0, const [0,1]}
       Rotation.X {type=MovieSceneScriptingDoubleChannel, default=0, const [0,1]}
       Rotation.Y {type=MovieSceneScriptingDoubleChannel, default=0, const [0,1]}
       Rotation.Z {type=MovieSceneScriptingDoubleChannel, default=0, const [0,1]}
       Scale.X {type=MovieSceneScriptingDoubleChannel, default=1, const [1,1]}
       Scale.Y {type=MovieSceneScriptingDoubleChannel, default=1, const [1,1]}
       Scale.Z {type=MovieSceneScriptingDoubleChannel, default=1, const [1,1]}
      ]
     }]
   }
   {type=MovieSceneSkeletalAnimationTrack, sections []}
   {
    type=MovieSceneControlRigParameterTrack
    display_name=Fortnite_Biped_CtrlRig
    sections [
     {
      type=MovieSceneControlRigParameterSection
      channels [
       leg_l_stretch_switch {type=MovieSceneScriptingBoolChannel, default=false, const [true,21]}
       leg_r_stretch_switch {type=MovieSceneScriptingBoolChannel, default=false, const [true,21]}
       arm_l_fk_local {type=MovieSceneScriptingBoolChannel, default=true, const [true,28]}
       Auto Interpolate_3 {type=MovieSceneScriptingBoolChannel, default=false, const [false,21]}
       Auto Interpolate {type=MovieSceneScriptingBoolChannel, default=false, const [false,21]}
       arm_l_stretch_switch {type=MovieSceneScriptingBoolChannel, default=false, const [false,24]}
       Keep Aligned {type=MovieSceneScriptingBoolChannel, default=false, const [false,24]}
       arm_r_fk_local {type=MovieSceneScriptingBoolChannel, default=true, const [true,31]}
       Auto Interpolate_4 {type=MovieSceneScriptingBoolChannel, default=false, const [false,21]}
       Auto Interpolate_2 {type=MovieSceneScriptingBoolChannel, default=false, const [false,21]}
       arm_r_stretch_switch {type=MovieSceneScriptingBoolChannel, default=false, const [false,24]}
       Keep Aligned_2 {type=MovieSceneScriptingBoolChannel, default=false, const [false,24]}
       head_local {type=MovieSceneScriptingBoolChannel, default=true, const [false,7]}
       Stretch {type=MovieSceneScriptingBoolChannel, default=false, const [true,7]}
       head_stretch {type=MovieSceneScriptingBoolChannel, default=false, const [true,27]}
       spine_stretch_switch {type=MovieSceneScriptingBoolChannel, default=false, const [false,27]}
       arm_l_fk_ik_switch {type=MovieSceneScriptingBoolChannel, default=false, const [false,21]}
       leg_l_fk_ik_switch {type=MovieSceneScriptingBoolChannel, default=true, const [true,21]}
       arm_r_fk_ik_switch {type=MovieSceneScriptingBoolChannel, default=false, const [false,21]}
       leg_r_fk_ik_switch {type=MovieSceneScriptingBoolChannel, default=true, const [true,21]}
       ShowBodyControls {type=MovieSceneScriptingBoolChannel, default=true, const [true,21]}
       spine_fk_ik_switch {type=MovieSceneScriptingBoolChannel, default=false, const [false,21]}
       neck_fk_ik_switch {type=MovieSceneScriptingBoolChannel, default=false, const [false,21]}
       space {type=MovieSceneScriptingByteChannel, default=1, const [0,7]}
       global_ctrl.Location.X {type=MovieSceneScriptingFloatChannel, default=0, const [0,22]}
       global_ctrl.Location.Y {type=MovieSceneScriptingFloatChannel, default=0, const [0,22]}
       global_ctrl.Location.Z {type=MovieSceneScriptingFloatChannel, default=0, const [0,22]}
       global_ctrl.Rotation.X {type=MovieSceneScriptingFloatChannel, default=0, const [0,22]}
       global_ctrl.Rotation.Y {type=MovieSceneScriptingFloatChannel, default=0, const [0,22]}
       global_ctrl.Rotation.Z {type=MovieSceneScriptingFloatChannel, default=0, const [0,22]}
       root_ctrl.Location.X {type=MovieSceneScriptingFloatChannel, default=0, const [0,22]}
       root_ctrl.Location.Y {type=MovieSceneScriptingFloatChannel, default=0, const [0,22]}
       root_ctrl.Location.Z {type=MovieSceneScriptingFloatChannel, default=0, const [0,22]}
       root_ctrl.Rotation.X {type=MovieSceneScriptingFloatChannel, default=0, const [0,22]}
       root_ctrl.Rotation.Y {type=MovieSceneScriptingFloatChannel, default=0, const [0,22]}
       root_ctrl.Rotation.Z {type=MovieSceneScriptingFloatChannel, default=0, const [0,22]}
       root_ctrl.Scale.X {type=MovieSceneScriptingFloatChannel, default=1, const [1,22]}
       root_ctrl.Scale.Y {type=MovieSceneScriptingFloatChannel, default=1, const [1,22]}
       root_ctrl.Scale.Z {type=MovieSceneScriptingFloatChannel, default=1, const [1,22]}
       body_offset_ctrl.Location.X {type=MovieSceneScriptingFloatChannel, default=0, thinned_from=22, keys [[0,7.0464],[294,7.0464]]}
       body_offset_ctrl.Location.Y {type=MovieSceneScriptingFloatChannel, default=0, const [0,22]}
       body_offset_ctrl.Location.Z {type=MovieSceneScriptingFloatChannel, default=0, const [0,22]}
       body_offset_ctrl.Rotation.X {type=MovieSceneScriptingFloatChannel, default=0, const [0,22]}
       body_offset_ctrl.Rotation.Y {type=MovieSceneScriptingFloatChannel, default=0, const [0,22]}
       body_offset_ctrl.Rotation.Z {type=MovieSceneScriptingFloatChannel, default=0, const [0,22]}
       body_offset_ctrl.Scale.X {type=MovieSceneScriptingFloatChannel, default=1, const [1,22]}
       body_offset_ctrl.Scale.Y {type=MovieSceneScriptingFloatChannel, default=1, const [1,22]}
       body_offset_ctrl.Scale.Z {type=MovieSceneScriptingFloatChannel, default=1, const [1,22]}
       thigh_l_fk_ctrl.Location.X {type=MovieSceneScriptingFloatChannel, default=0, thinned_from=22, keys [[0,0],[294,0]]}
       thigh_l_fk_ctrl.Location.Y {type=MovieSceneScriptingFloatChannel, default=0, thinned_from=22, keys [[0,0],[294,0]]}
       thigh_l_fk_ctrl.Location.Z {type=MovieSceneScriptingFloatChannel, default=0, thinned_from=22, keys [[0,0],[294,0]]}
       thigh_l_fk_ctrl.Rotation.X {
        type=MovieSceneScriptingFloatChannel
        default=1.216
        keys [[0,13.2847],[25,13.454],[42,10.4208],[50,5.7514],[58,4.8744],[68,4.6806],[76,4.8744],[87,6.6893],[97,8.6318],[111,6.4978],[127,9.8579],[134,13.834],[142,15.1605],[170,12.7805],[174,12.0568],[177,12.8522],[185,12.8517],[199,13.2268],[216,13.7732],[248,13.195],[269,13.2847],[294,13.454]]
       }
       thigh_l_fk_ctrl.Rotation.Y {
        type=MovieSceneScriptingFloatChannel
        default=-1.2128
        keys [[0,17.587],[25,17.6865],[42,16.255],[50,14.8178],[58,14.6535],[68,15.6656],[76,14.6535],[87,14.2925],[97,15.5885],[111,17.2371],[127,15.8391],[134,16.4771],[142,16.5985],[169,17.1303],[173,18.1882],[176,16.8028],[184,17.1861],[199,17.6254],[216,17.4419],[248,17.6467],[269,17.587],[294,17.6865]]
       }
       thigh_l_fk_ctrl.Rotation.Z {
        type=MovieSceneScriptingFloatChannel
        default=-0.1664
        keys [[0,2.7024],[25,3.1102],[42,2.3526],[50,-1.9864],[58,-2.4542],[68,-2.2166],[76,-2.4542],[87,-1.9673],[97,-1.7156],[111,-1.8844],[127,0.6303],[134,4.5346],[142,5.9959],[169,4.6293],[173,3.6241],[176,4.6207],[184,2.5659],[199,2.7259],[216,4.185],[248,2.7244],[269,2.7024],[294,3.1102]]
       }
       thigh_l_fk_ctrl.Scale.X {type=MovieSceneScriptingFloatChannel, default=1, const [1,22]}
       thigh_l_fk_ctrl.Scale.Y {type=MovieSceneScriptingFloatChannel, default=1, const [1,22]}
       thigh_l_fk_ctrl.Scale.Z {type=MovieSceneScriptingFloatChannel, default=1, const [1,22]}
       calf_l_fk_ctrl.Location.X {
        type=MovieSceneScriptingFloatChannel
        default=0
        thinned_from=22
        keys [[0,-0.15],[25,-0.1465],[42,0],[184,0],[199,-0.148],[216,0],[248,-0.1481],[269,-0.1501],[294,-0.1465]]
       }
       calf_l_fk_ctrl.Location.Y {
        type=MovieSceneScriptingFloatChannel
        default=0
        thinned_from=22
        keys [[0,0.0124],[25,0.011],[42,0],[184,0],[199,0.0111],[216,0],[248,0.005],[269,0.0112],[294,0.011]]
       }
       calf_l_fk_ctrl.Location.Z {
        type=MovieSceneScriptingFloatChannel
        default=0
        thinned_from=22
        keys [[0,-0.0026],[25,-0.0031],[42,0],[184,0],[199,-0.0032],[216,0],[248,-0.0086],[269,-0.0032],[294,-0.0031]]
       }
       calf_l_fk_ctrl.Rotation.X {
        type=MovieSceneScriptingFloatChannel
        default=0.0543
        keys [[0,0.4904],[25,0.4926],[42,0.4422],[50,0.4785],[58,0.477],[68,0.4312],[76,0.477],[87,0.5002],[97,0.578],[111,0.5279],[127,0.5349],[134,0.469],[142,0.4418],[170,0.4701],[174,0.4575],[177,0.4379],[185,0.4488],[199,0.4919],[216,0.454],[248,0.4923],[269,0.491],[294,0.4926]]
       }
       calf_l_fk_ctrl.Rotation.Y {
        type=MovieSceneScriptingFloatChannel
        default=1.2238
        keys [[0,1.1216],[25,1.1216],[42,1.1424],[50,1.1292],[58,1.1298],[68,1.1466],[76,1.1298],[87,1.1211],[97,1.0882],[111,1.1054],[127,1.1072],[134,1.1326],[142,1.1419],[170,1.1312],[174,1.1364],[177,1.1441],[185,1.1398],[199,1.1219],[216,1.1378],[248,1.1305],[269,1.1223],[294,1.1216]]
       }
       calf_l_fk_ctrl.Rotation.Z {
        type=MovieSceneScriptingFloatChannel
        default=0.3732
        keys [[0,21.464],[25,21.5437],[42,18.9949],[50,20.803],[58,20.7269],[68,18.4447],[76,20.7269],[87,21.8839],[97,25.8392],[111,23.3621],[127,23.6327],[134,20.3298],[142,18.9866],[170,20.4031],[174,19.7668],[177,18.7777],[185,19.3268],[199,21.5121],[216,19.5878],[248,21.5262],[269,21.4657],[294,21.5437]]
       }
       calf_l_fk_ctrl.Scale.X {type=MovieSceneScriptingFloatChannel, default=1, const [1,22]}
       calf_l_fk_ctrl.Scale.Y {type=MovieSceneScriptingFloatChannel, default=1, const [1,22]}
       calf_l_fk_ctrl.Scale.Z {type=MovieSceneScriptingFloatChannel, default=1, const [1,22]}
       foot_l_fk_ctrl.Location.X {
        type=MovieSceneScriptingFloatChannel
        default=0
        thinned_from=22
        keys [[0,-0.164],[25,-0.1602],[42,0],[185,0],[199,-0.1617],[216,0],[248,-0.1615],[269,-0.164],[294,-0.1602]]
       }
       foot_l_fk_ctrl.Location.Y {type=MovieSceneScriptingFloatChannel, default=0.0398, thinned_from=22, keys [[0,0.0398],[294,0.0398]]}
       foot_l_fk_ctrl.Location.Z {type=MovieSceneScriptingFloatChannel, default=-0.8885, thinned_from=22, keys [[0,-0.8885],[294,-0.8885]]}
       foot_l_fk_ctrl.Rotation.X {
        type=MovieSceneScriptingFloatChannel
        default=-1.1647
        thinned_from=22
        keys [[0,8.8266],[25,8.6195],[42,8.6304],[50,8.9812],[58,9.0077],[68,8.7061],[76,9.0077],[87,9.0329],[97,9.1643],[111,8.9466],[127,9.3837],[134,9.6954],[142,9.8068],[170,8.6945],[174,8.6508],[177,8.6463],[185,8.5943],[216,8.6142],[248,8.6857],[269,8.8266],[294,8.6195]]
       }
       foot_l_fk_ctrl.Rotation.Y {
        type=MovieSceneScriptingFloatChannel
        default=0.0792
        keys [[0,-9.3105],[25,-9.3316],[42,-9.4114],[50,-8.6379],[58,-8.6161],[68,-9.3112],[76,-8.6161],[87,-8.1199],[97,-7.0254],[111,-7.7608],[127,-7.6978],[134,-8.4255],[142,-8.6661],[169,-9.2174],[173,-9.5548],[176,-9.4311],[184,-9.4364],[199,-9.3541],[216,-9.3258],[248,-9.3635],[269,-9.3105],[294,-9.3316]]
       }
       foot_l_fk_ctrl.Rotation.Z {
        type=MovieSceneScriptingFloatChannel
        default=-0.2074
        keys [[0,-9.7297],[25,-9.7917],[42,-8.6707],[50,-10.5573],[58,-10.6087],[68,-8.6858],[76,-10.6087],[87,-11.1538],[97,-13.3244],[111,-11.6113],[127,-13.1975],[134,-12.7969],[142,-12.5504],[169,-9.5437],[173,-9.1688],[176,-8.6382],[184,-8.6901],[199,-9.7676],[216,-8.8589],[248,-9.7843],[269,-9.7297],[294,-9.7917]]
       }
       foot_l_fk_ctrl.Scale.X {type=MovieSceneScriptingFloatChannel, default=1, const [1,22]}
       foot_l_fk_ctrl.Scale.Y {type=MovieSceneScriptingFloatChannel, default=1, const [1,22]}
       foot_l_fk_ctrl.Scale.Z {type=MovieSceneScriptingFloatChannel, default=1, const [1,22]}
       ball_l_fk_ctrl.Location.X {type=MovieSceneScriptingFloatChannel, default=0, const [0,22]}
       ball_l_fk_ctrl.Location.Y {type=MovieSceneScriptingFloatChannel, default=0, const [0,22]}
       ball_l_fk_ctrl.Location.Z {type=MovieSceneScriptingFloatChannel, default=0, const [0,22]}
       ball_l_fk_ctrl.Rotation.X {type=MovieSceneScriptingFloatChannel, default=0, thinned_from=22, keys [[0,0],[294,0]]}
       ball_l_fk_ctrl.Rotation.Y {type=MovieSceneScriptingFloatChannel, default=0, thinned_from=22, keys [[0,0],[294,0]]}
       ball_l_fk_ctrl.Rotation.Z {type=MovieSceneScriptingFloatChannel, default=0, thinned_from=22, keys [[0,0],[294,0]]}
       ball_l_fk_ctrl.Scale.X {type=MovieSceneScriptingFloatChannel, default=1, const [1,22]}
       ball_l_fk_ctrl.Scale.Y {type=MovieSceneScriptingFloatChannel, default=1, const [1,22]}
       ball_l_fk_ctrl.Scale.Z {type=MovieSceneScriptingFloatChannel, default=1, const [1,22]}
       foot_l_ik_ctrl.Location.X {type=MovieSceneScriptingFloatChannel, default=0, const [1.4784,22]}
       foot_l_ik_ctrl.Location.Y {type=MovieSceneScriptingFloatChannel, default=0, const [13.7508,22]}
       foot_l_ik_ctrl.Location.Z {type=MovieSceneScriptingFloatChannel, default=0, const [0,22]}
       foot_l_ik_ctrl.Rotation.X {type=MovieSceneScriptingFloatChannel, default=0, const [0,22]}
       foot_l_ik_ctrl.Rotation.Y {type=MovieSceneScriptingFloatChannel, default=0, const [0,22]}
       foot_l_ik_ctrl.Rotation.Z {type=MovieSceneScriptingFloatChannel, default=0, const [17.1216,22]}
       foot_l_ik_ctrl.Scale.X {type=MovieSceneScriptingFloatChannel, default=1, const [1,22]}
       foot_l_ik_ctrl.Scale.Y {type=MovieSceneScriptingFloatChannel, default=1, const [1,22]}
       foot_l_ik_ctrl.Scale.Z {type=MovieSceneScriptingFloatChannel, default=1, const [1,22]}
       ball_l_ik_ctrl.Location.X {type=MovieSceneScriptingFloatChannel, default=0, const [0,22]}
       ball_l_ik_ctrl.Location.Y {type=MovieSceneScriptingFloatChannel, default=0, const [0,22]}
       ball_l_ik_ctrl.Location.Z {type=MovieSceneScriptingFloatChannel, default=0, const [0,22]}
       ball_l_ik_ctrl.Rotation.X {type=MovieSceneScriptingFloatChannel, default=0, const [0,22]}
       ball_l_ik_ctrl.Rotation.Y {type=MovieSceneScriptingFloatChannel, default=0, const [0,22]}
       ball_l_ik_ctrl.Rotation.Z {type=MovieSceneScriptingFloatChannel, default=0, const [0,22]}
       ball_l_ik_ctrl.Scale.X {type=MovieSceneScriptingFloatChannel, default=1, const [1,22]}
       ball_l_ik_ctrl.Scale.Y {type=MovieSceneScriptingFloatChannel, default=1, const [1,22]}
       ball_l_ik_ctrl.Scale.Z {type=MovieSceneScriptingFloatChannel, default=1, const [1,22]}
       foot_bk1_l_ctrl.Location.X {type=MovieSceneScriptingFloatChannel, default=0, const [0,22]}
       foot_bk1_l_ctrl.Location.Y {type=MovieSceneScriptingFloatChannel, default=0, const [0,22]}
       foot_bk1_l_ctrl.Location.Z {type=MovieSceneScriptingFloatChannel, default=0, const [0,22]}
       foot_bk1_l_ctrl.Rotation.X {type=MovieSceneScriptingFloatChannel, default=0, const [0,22]}
       foot_bk1_l_ctrl.Rotation.Y {type=MovieSceneScriptingFloatChannel, default=0, const [0,22]}
       foot_bk1_l_ctrl.Rotation.Z {type=MovieSceneScriptingFloatChannel, default=0, const [0,22]}
       foot_bk1_l_ctrl.Scale.X {type=MovieSceneScriptingFloatChannel, default=1, const [1,22]}
       foot_bk1_l_ctrl.Scale.Y {type=MovieSceneScriptingFloatChannel, default=1, const [1,22]}
       foot_bk1_l_ctrl.Scale.Z {type=MovieSceneScriptingFloatChannel, default=1, const [1,22]}
       foot_roll_l_ctrl.Location.X {type=MovieSceneScriptingFloatChannel, default=0, const [0,22]}
       foot_roll_l_ctrl.Location.Y {type=MovieSceneScriptingFloatChannel, default=0, const [0,22]}
       foot_roll_l_ctrl.Location.Z {type=MovieSceneScriptingFloatChannel, default=0, const [0,22]}
       foot_roll_l_ctrl.Rotation.X {type=MovieSceneScriptingFloatChannel, default=0, const [0,22]}
       foot_roll_l_ctrl.Rotation.Y {type=MovieSceneScriptingFloatChannel, default=0, const [0,22]}
       foot_roll_l_ctrl.Rotation.Z {type=MovieSceneScriptingFloatChannel, default=0, const [0,22]}
       foot_roll_l_ctrl.Scale.X {type=MovieSceneScriptingFloatChannel, default=1, const [1,22]}
       foot_roll_l_ctrl.Scale.Y {type=MovieSceneScriptingFloatChannel, default=1, const [1,22]}
       foot_roll_l_ctrl.Scale.Z {type=MovieSceneScriptingFloatChannel, default=1, const [1,22]}
       roll_blend_l_ctrl {type=MovieSceneScriptingFloatChannel, default=1, const [1,22]}
       tip_l_ctrl.Location.X {type=MovieSceneScriptingFloatChannel, default=0, const [0,22]}
       tip_l_ctrl.Location.Y {type=MovieSceneScriptingFloatChannel, default=0, const [0,22]}
       tip_l_ctrl.Location.Z {type=MovieSceneScriptingFloatChannel, default=0, const [0,22]}
       tip_l_ctrl.Rotation.X {type=MovieSceneScriptingFloatChannel, default=0, const [0,22]}
       tip_l_ctrl.Rotation.Y {type=MovieSceneScriptingFloatChannel, default=0, const [0,22]}
       tip_l_ctrl.Rotation.Z {type=MovieSceneScriptingFloatChannel, default=0, const [0,22]}
       tip_l_ctrl.Scale.X {type=MovieSceneScriptingFloatChannel, default=1, const [1,22]}
       tip_l_ctrl.Scale.Y {type=MovieSceneScriptingFloatChannel, default=1, const [1,22]}
       tip_l_ctrl.Scale.Z {type=MovieSceneScriptingFloatChannel, default=1, const [1,22]}
       heel_l_ctrl.Location.X {type=MovieSceneScriptingFloatChannel, default=0, const [0,22]}
       heel_l_ctrl.Location.Y {type=MovieSceneScriptingFloatChannel, default=0, const [0,22]}
       heel_l_ctrl.Location.Z {type=MovieSceneScriptingFloatChannel, default=0, const [0,22]}
       heel_l_ctrl.Rotation.X {type=MovieSceneScriptingFloatChannel, default=0, const [0,22]}
       heel_l_ctrl.Rotation.Y {type=MovieSceneScriptingFloatChannel, default=0, const [0,22]}
       heel_l_ctrl.Rotation.Z {type=MovieSceneScriptingFloatChannel, default=0, const [0,22]}
       heel_l_ctrl.Scale.X {type=MovieSceneScriptingFloatChannel, default=1, const [1,22]}
       heel_l_ctrl.Scale.Y {type=MovieSceneScriptingFloatChannel, default=1, const [1,22]}
       heel_l_ctrl.Scale.Z {type=MovieSceneScriptingFloatChannel, default=1, const [1,22]}
       leg_l_softness {type=MovieSceneScriptingFloatChannel, default=0, const [4.8668,22]}
       leg_l_pv_ik_ctrl.Location.X {
        type=MovieSceneScriptingFloatChannel
        default=-7.7499
        thinned_from=20
        keys [[0,-54.2217],[25,-54.3412],[32,-54.816],[42,-54.472],[50,-53.4722],[68,-52.9621],[97,-53.2088],[111,-53.6241],[119,-53.7129],[127,-53.0721],[134,-51.6992],[142,-51.5295],[159,-53.253],[166,-54.4203],[173,-54.8161],[199,-54.2217],[216,-54.3412],[269,-54.2217],[294,-54.3412]]
       }
       leg_l_pv_ik_ctrl.Location.Y {
        type=MovieSceneScriptingFloatChannel
        default=-0.0319
        thinned_from=20
        keys [[0,-12.9709],[25,-13.3582],[32,-13.8258],[42,-10.5836],[50,-8.5469],[68,-7.6199],[97,-9.3467],[111,-10.3697],[127,-10.9366],[134,-11.9239],[142,-12.0971],[159,-14.3504],[166,-13.0945],[173,-12.7392],[199,-12.9709],[216,-13.3582],[248,-12.9709],[269,-12.9709],[294,-13.3582]]
       }
       leg_l_pv_ik_ctrl.Location.Z {
        type=MovieSceneScriptingFloatChannel
        default=-0.7162
        keys [[0,1.1782],[25,1.0014],[32,1.2379],[42,3.813],[50,4.6214],[68,4.9829],[97,3.5313],[111,2.4856],[119,1.8733],[127,1.0547],[134,-0.8345],[142,-1.1567],[159,-1.1483],[166,1.9634],[172,2.1069],[199,1.1782],[216,1.0014],[248,1.1782],[269,1.1782],[294,1.0014]]
       }
       leg_l_pv_ik_ctrl.Rotation.X {
        type=MovieSceneScriptingFloatChannel
        default=0
        thinned_from=20
        keys [[0,0.5956],[25,0.8927],[32,0.6465],[42,-4.6643],[50,-7.2606],[68,-8.7751],[97,-6.1395],[111,-4.6204],[127,-3.4033],[134,-1.2596],[142,0.2489],[159,2.1555],[166,0.1441],[173,-0.5734],[199,0.5956],[216,0.8927],[248,0.5956],[269,0.5956],[294,0.8927]]
       }
       leg_l_pv_ik_ctrl.Rotation.Y {
        type=MovieSceneScriptingFloatChannel
        default=0
        keys [[0,2.7747],[25,2.7823],[32,2.3305],[42,2.1927],[50,2.4829],[68,2.3204],[97,2.7481],[111,3.1292],[119,3.7795],[127,4.9244],[134,5.6654],[142,4.3745],[159,5.7042],[166,3.2051],[173,2.9324],[199,2.7747],[216,2.7823],[248,2.7747],[269,2.7747],[294,2.7823]]
       }
       leg_l_pv_ik_ctrl.Rotation.Z {
        type=MovieSceneScriptingFloatChannel
        default=0
        keys [[0,8.464],[25,8.79],[32,9.4249],[42,8.1876],[50,6.7387],[68,6.0506],[97,6.9344],[111,7.5379],[119,7.8064],[127,7.9284],[134,7.7363],[142,6.4643],[159,9.9468],[166,9.5708],[173,9.2332],[199,8.464],[216,8.79],[248,8.464],[269,8.464],[294,8.79]]
       }
       leg_l_pv_ik_ctrl.Scale.X {type=MovieSceneScriptingFloatChannel, default=1, const [1,20]}
       leg_l_pv_ik_ctrl.Scale.Y {type=MovieSceneScriptingFloatChannel, default=1, const [1,20]}
       leg_l_pv_ik_ctrl.Scale.Z {type=MovieSceneScriptingFloatChannel, default=1, const [1,20]}
       thigh_r_fk_ctrl.Location.X {type=MovieSceneScriptingFloatChannel, default=0, thinned_from=22, keys [[0,0],[294,0]]}
       thigh_r_fk_ctrl.Location.Y {type=MovieSceneScriptingFloatChannel, default=0, thinned_from=22, keys [[0,0],[294,0]]}
       thigh_r_fk_ctrl.Location.Z {type=MovieSceneScriptingFloatChannel, default=0, thinned_from=22, keys [[0,0],[294,0]]}
       thigh_r_fk_ctrl.Rotation.X {
        type=MovieSceneScriptingFloatChannel
        default=1.2139
        keys [[0,19.8357],[25,19.5031],[42,22.2541],[50,27.8213],[58,28.7941],[68,28.1885],[76,28.7941],[87,26.8995],[97,24.901],[111,25.2521],[127,23.2419],[134,18.9473],[142,17.5544],[170,19.4255],[174,20.9504],[176,19.7737],[183,20.6573],[199,19.7732],[216,19.484],[248,19.8577],[269,19.8357],[294,19.5031]]
       }
       thigh_r_fk_ctrl.Rotation.Y {
        type=MovieSceneScriptingFloatChannel
        default=-1.213
        keys [[0,-0.3266],[25,-0.386],[42,1.7511],[50,1.6455],[58,1.791],[68,1.1676],[76,1.791],[87,1.8613],[97,-0.8781],[111,-0.1624],[127,-0.1421],[134,-0.2161],[142,-0.3054],[170,1.2624],[174,-1.167],[176,1.3336],[183,-0.137],[199,-0.3293],[216,-0.3844],[248,-0.3456],[269,-0.3266],[294,-0.386]]
       }
       thigh_r_fk_ctrl.Rotation.Z {
        type=MovieSceneScriptingFloatChannel
        default=-0.1592
        keys [[0,3.4427],[25,3.8258],[42,6.8358],[50,2.2553],[58,2.5873],[68,6.8602],[76,2.5873],[87,2.4767],[97,5.145],[111,8.853],[127,3.848],[134,1.5146],[142,0.8349],[170,7.9627],[174,4.7158],[177,7.2683],[185,2.7232],[199,3.5808],[216,3.8617],[248,3.5721],[269,3.4427],[294,3.8258]]
       }
       thigh_r_fk_ctrl.Scale.X {type=MovieSceneScriptingFloatChannel, default=1, const [1,22]}
       thigh_r_fk_ctrl.Scale.Y {type=MovieSceneScriptingFloatChannel, default=1, const [1,22]}
       thigh_r_fk_ctrl.Scale.Z {type=MovieSceneScriptingFloatChannel, default=1, const [1,22]}
       calf_r_fk_ctrl.Location.X {
        type=MovieSceneScriptingFloatChannel
        default=0
        thinned_from=22
        keys [[0,-0.3836],[42,-0.3737],[216,-0.3737],[248,-0.3746],[269,-0.3837],[294,-0.3776]]
       }
       calf_r_fk_ctrl.Location.Y {
        type=MovieSceneScriptingFloatChannel
        default=0
        thinned_from=22
        keys [[0,0.03],[25,0.0289],[216,0.0286],[248,0.0287],[269,0.0293],[294,0.0289]]
       }
       calf_r_fk_ctrl.Location.Z {type=MovieSceneScriptingFloatChannel, default=0, thinned_from=22, keys [[0,-0.008],[248,-0.0079],[269,-0.0081],[294,-0.0079]]}
       calf_r_fk_ctrl.Rotation.X {
        type=MovieSceneScriptingFloatChannel
        default=0.0561
        keys [[0,0.3736],[25,0.3738],[42,0.3343],[50,0.4013],[58,0.398],[68,0.3045],[76,0.398],[87,0.3816],[97,0.2862],[111,0.2594],[127,0.3491],[134,0.4035],[142,0.419],[170,0.3317],[174,0.3664],[177,0.3321],[185,0.3783],[199,0.3747],[216,0.3736],[248,0.3745],[269,0.3722],[294,0.3738]]
       }
       calf_r_fk_ctrl.Rotation.Y {
        type=MovieSceneScriptingFloatChannel
        default=1.1982
        thinned_from=22
        keys [[0,1.1397],[25,1.1398],[42,1.152],[50,1.1202],[58,1.1219],[68,1.1602],[76,1.1219],[87,1.1299],[97,1.1707],[111,1.1711],[127,1.1449],[134,1.1192],[142,1.1112],[170,1.1527],[174,1.1422],[177,1.1526],[185,1.1383],[294,1.1398]]
       }
       calf_r_fk_ctrl.Rotation.Z {
        type=MovieSceneScriptingFloatChannel
        default=0.2616
        keys [[0,15.6592],[25,15.738],[42,13.7648],[50,17.3062],[58,17.1288],[68,12.2881],[76,17.1288],[87,16.254],[97,11.2975],[111,10.071],[127,14.5464],[134,17.4209],[142,18.2558],[170,13.6349],[174,15.3708],[176,13.6566],[183,15.9682],[199,15.7874],[216,15.7285],[248,15.7758],[269,15.6602],[294,15.738]]
       }
       calf_r_fk_ctrl.Scale.X {type=MovieSceneScriptingFloatChannel, default=1, const [1,22]}
       calf_r_fk_ctrl.Scale.Y {type=MovieSceneScriptingFloatChannel, default=1, const [1,22]}
       calf_r_fk_ctrl.Scale.Z {type=MovieSceneScriptingFloatChannel, default=1, const [1,22]}
       foot_r_fk_ctrl.Location.X {
        type=MovieSceneScriptingFloatChannel
        default=0
        thinned_from=22
        keys [[0,-0.4194],[42,-0.4085],[216,-0.4085],[248,-0.4095],[269,-0.4194],[294,-0.4127]]
       }
       foot_r_fk_ctrl.Location.Y {type=MovieSceneScriptingFloatChannel, default=-0.0329, thinned_from=22, keys [[0,-0.0329],[294,-0.0329]]}
       foot_r_fk_ctrl.Location.Z {type=MovieSceneScriptingFloatChannel, default=-0.9075, thinned_from=22, keys [[0,-0.9075],[294,-0.9075]]}
       foot_r_fk_ctrl.Rotation.X {
        type=MovieSceneScriptingFloatChannel
        default=-1.1648
        keys [[0,-2.1652],[25,-1.8889],[42,-1.6293],[50,-1.5786],[58,-1.5467],[68,-1.3683],[76,-1.5467],[87,-1.7123],[97,-2.0087],[111,-1.704],[127,-1.725],[134,-1.4792],[142,-1.3999],[169,-1.7516],[173,-1.7705],[176,-1.7104],[182,-1.8273],[199,-1.8642],[216,-1.8766],[248,-1.9876],[269,-2.1652],[294,-1.8889]]
       }
       foot_r_fk_ctrl.Rotation.Y {
        type=MovieSceneScriptingFloatChannel
        default=0.1067
        thinned_from=22
        keys [[0,-5.6615],[25,-5.703],[42,-8.0822],[50,-5.7655],[58,-5.9142],[68,-8.7898],[76,-5.9142],[87,-6.1403],[97,-6.9079],[111,-10.0389],[127,-6.622],[134,-6.1103],[142,-5.9462],[169,-8.1677],[173,-5.7985],[176,-7.913],[184,-5.6792],[199,-5.7014],[216,-5.709],[269,-5.6615],[294,-5.703]]
       }
       foot_r_fk_ctrl.Rotation.Z {
        type=MovieSceneScriptingFloatChannel
        default=-0.1036
        thinned_from=22
        keys [[0,-4.7112],[25,-4.8326],[42,-5.7104],[50,-6.4155],[58,-6.561],[68,-6.1914],[76,-6.561],[87,-5.595],[97,-2.5091],[111,-3.9871],[127,-4.9949],[134,-7.0788],[142,-7.7186],[169,-5.1469],[173,-5.2302],[176,-5.2768],[199,-4.9156],[216,-4.8269],[248,-4.8951],[269,-4.7112],[294,-4.8326]]
       }
       foot_r_fk_ctrl.Scale.X {type=MovieSceneScriptingFloatChannel, default=1, const [1,22]}
       foot_r_fk_ctrl.Scale.Y {type=MovieSceneScriptingFloatChannel, default=1, const [1,22]}
       foot_r_fk_ctrl.Scale.Z {type=MovieSceneScriptingFloatChannel, default=1, const [1,22]}
       ball_r_fk_ctrl.Location.X {type=MovieSceneScriptingFloatChannel, default=0, const [0,22]}
       ball_r_fk_ctrl.Location.Y {type=MovieSceneScriptingFloatChannel, default=0, const [0,22]}
       ball_r_fk_ctrl.Location.Z {type=MovieSceneScriptingFloatChannel, default=0, const [0,22]}
       ball_r_fk_ctrl.Rotation.X {type=MovieSceneScriptingFloatChannel, default=0, thinned_from=22, keys [[0,0],[294,0]]}
       ball_r_fk_ctrl.Rotation.Y {type=MovieSceneScriptingFloatChannel, default=0, thinned_from=22, keys [[0,0],[294,0]]}
       ball_r_fk_ctrl.Rotation.Z {type=MovieSceneScriptingFloatChannel, default=0, thinned_from=22, keys [[0,0],[294,0]]}
       ball_r_fk_ctrl.Scale.X {type=MovieSceneScriptingFloatChannel, default=1, const [1,22]}
       ball_r_fk_ctrl.Scale.Y {type=MovieSceneScriptingFloatChannel, default=1, const [1,22]}
       ball_r_fk_ctrl.Scale.Z {type=MovieSceneScriptingFloatChannel, default=1, const [1,22]}
       foot_r_ik_ctrl.Location.X {type=MovieSceneScriptingFloatChannel, default=0, const [5.103,22]}
       foot_r_ik_ctrl.Location.Y {type=MovieSceneScriptingFloatChannel, default=0, const [-15.8664,22]}
       foot_r_ik_ctrl.Location.Z {type=MovieSceneScriptingFloatChannel, default=0, const [0,22]}
       foot_r_ik_ctrl.Rotation.X {type=MovieSceneScriptingFloatChannel, default=0, const [0,22]}
       foot_r_ik_ctrl.Rotation.Y {type=MovieSceneScriptingFloatChannel, default=0, const [0,22]}
       foot_r_ik_ctrl.Rotation.Z {type=MovieSceneScriptingFloatChannel, default=0, const [-63.0692,22]}
       foot_r_ik_ctrl.Scale.X {type=MovieSceneScriptingFloatChannel, default=1, const [1,22]}
       foot_r_ik_ctrl.Scale.Y {type=MovieSceneScriptingFloatChannel, default=1, const [1,22]}
       foot_r_ik_ctrl.Scale.Z {type=MovieSceneScriptingFloatChannel, default=1, const [1,22]}
       ball_r_ik_ctrl.Location.X {type=MovieSceneScriptingFloatChannel, default=0, const [0,22]}
       ball_r_ik_ctrl.Location.Y {type=MovieSceneScriptingFloatChannel, default=0, const [0,22]}
       ball_r_ik_ctrl.Location.Z {type=MovieSceneScriptingFloatChannel, default=0, const [0,22]}
       ball_r_ik_ctrl.Rotation.X {type=MovieSceneScriptingFloatChannel, default=0, thinned_from=22, keys [[0,0],[294,0]]}
       ball_r_ik_ctrl.Rotation.Y {type=MovieSceneScriptingFloatChannel, default=0, thinned_from=22, keys [[0,0],[294,0]]}
       ball_r_ik_ctrl.Rotation.Z {type=MovieSceneScriptingFloatChannel, default=0, thinned_from=22, keys [[0,0],[294,0]]}
       ball_r_ik_ctrl.Scale.X {type=MovieSceneScriptingFloatChannel, default=1, const [1,22]}
       ball_r_ik_ctrl.Scale.Y {type=MovieSceneScriptingFloatChannel, default=1, const [1,22]}
       ball_r_ik_ctrl.Scale.Z {type=MovieSceneScriptingFloatChannel, default=1, const [1,22]}
       foot_bk1_r_ctrl.Location.X {type=MovieSceneScriptingFloatChannel, default=0, const [0,22]}
       foot_bk1_r_ctrl.Location.Y {type=MovieSceneScriptingFloatChannel, default=0, const [0,22]}
       foot_bk1_r_ctrl.Location.Z {type=MovieSceneScriptingFloatChannel, default=0, const [0,22]}
       foot_bk1_r_ctrl.Rotation.X {type=MovieSceneScriptingFloatChannel, default=0, thinned_from=22, keys [[0,0],[294,0]]}
       foot_bk1_r_ctrl.Rotation.Y {type=MovieSceneScriptingFloatChannel, default=0, const [0,22]}
       foot_bk1_r_ctrl.Rotation.Z {type=MovieSceneScriptingFloatChannel, default=0, const [0,22]}
       foot_bk1_r_ctrl.Scale.X {type=MovieSceneScriptingFloatChannel, default=1, const [1,22]}
       foot_bk1_r_ctrl.Scale.Y {type=MovieSceneScriptingFloatChannel, default=1, const [1,22]}
       foot_bk1_r_ctrl.Scale.Z {type=MovieSceneScriptingFloatChannel, default=1, const [1,22]}
       foot_roll_r_ctrl.Location.X {type=MovieSceneScriptingFloatChannel, default=0, const [0,22]}
       foot_roll_r_ctrl.Location.Y {type=MovieSceneScriptingFloatChannel, default=0, const [0,22]}
       foot_roll_r_ctrl.Location.Z {type=MovieSceneScriptingFloatChannel, default=0, const [0,22]}
       foot_roll_r_ctrl.Rotation.X {type=MovieSceneScriptingFloatChannel, default=0, const [0,22]}
       foot_roll_r_ctrl.Rotation.Y {type=MovieSceneScriptingFloatChannel, default=0, const [0,22]}
       foot_roll_r_ctrl.Rotation.Z {type=MovieSceneScriptingFloatChannel, default=0, thinned_from=22, keys [[0,0],[294,0]]}
       foot_roll_r_ctrl.Scale.X {type=MovieSceneScriptingFloatChannel, default=1, const [1,22]}
       foot_roll_r_ctrl.Scale.Y {type=MovieSceneScriptingFloatChannel, default=1, const [1,22]}
       foot_roll_r_ctrl.Scale.Z {type=MovieSceneScriptingFloatChannel, default=1, const [1,22]}
       roll_blend_r_ctrl {type=MovieSceneScriptingFloatChannel, default=1, const [1,22]}
       tip_r_ctrl.Location.X {type=MovieSceneScriptingFloatChannel, default=0, const [0,22]}
       tip_r_ctrl.Location.Y {type=MovieSceneScriptingFloatChannel, default=0, const [0,22]}
       tip_r_ctrl.Location.Z {type=MovieSceneScriptingFloatChannel, default=0, const [0,22]}
       tip_r_ctrl.Rotation.X {type=MovieSceneScriptingFloatChannel, default=0, thinned_from=22, keys [[0,0],[294,0]]}
       tip_r_ctrl.Rotation.Y {type=MovieSceneScriptingFloatChannel, default=0, const [0,22]}
       tip_r_ctrl.Rotation.Z {type=MovieSceneScriptingFloatChannel, default=0, thinned_from=22, keys [[0,0],[294,0]]}
       tip_r_ctrl.Scale.X {type=MovieSceneScriptingFloatChannel, default=1, const [1,22]}
       tip_r_ctrl.Scale.Y {type=MovieSceneScriptingFloatChannel, default=1, const [1,22]}
       tip_r_ctrl.Scale.Z {type=MovieSceneScriptingFloatChannel, default=1, const [1,22]}
       heel_r_ctrl.Location.X {type=MovieSceneScriptingFloatChannel, default=0, thinned_from=22, keys [[0,0],[294,0]]}
       heel_r_ctrl.Location.Y {type=MovieSceneScriptingFloatChannel, default=0, thinned_from=22, keys [[0,0],[294,0]]}
       heel_r_ctrl.Location.Z {type=MovieSceneScriptingFloatChannel, default=0, thinned_from=22, keys [[0,0],[294,0]]}
       heel_r_ctrl.Rotation.X {type=MovieSceneScriptingFloatChannel, default=0, const [0,22]}
       heel_r_ctrl.Rotation.Y {type=MovieSceneScriptingFloatChannel, default=0, const [0,22]}
       heel_r_ctrl.Rotation.Z {type=MovieSceneScriptingFloatChannel, default=0, thinned_from=22, keys [[0,0],[294,0]]}
       heel_r_ctrl.Scale.X {type=MovieSceneScriptingFloatChannel, default=1, const [1,22]}
       heel_r_ctrl.Scale.Y {type=MovieSceneScriptingFloatChannel, default=1, const [1,22]}
       heel_r_ctrl.Scale.Z {type=MovieSceneScriptingFloatChannel, default=1, const [1,22]}
       leg_r_softness {type=MovieSceneScriptingFloatChannel, default=0, const [1.0418,22]}
       leg_r_pv_ik_ctrl.Location.X {
        type=MovieSceneScriptingFloatChannel
        default=0
        thinned_from=22
        keys [[0,50.0761],[25,50.2526],[42,49.6024],[50,48.9696],[58,48.7031],[76,48.4684],[87,48.9223],[97,49.6343],[111,49.5978],[127,50.4415],[134,50.608],[142,49.5405],[170,50.6482],[177,50.339],[185,49.5244],[197,50.0761],[216,50.2526],[248,50.0281],[269,50.0761],[294,50.2526]]
       }
       leg_r_pv_ik_ctrl.Location.Y {
        type=MovieSceneScriptingFloatChannel
        default=0
        keys [[0,-9.9662],[25,-10.135],[42,-11.8996],[50,-11.8263],[58,-11.7957],[68,-11.9234],[76,-12.0969],[87,-12.1549],[97,-11.5772],[111,-11.2463],[127,-11.1194],[134,-10.1282],[142,-8.4394],[170,-11.2031],[174,-11.1018],[177,-10.8161],[185,-9.4595],[197,-9.9662],[216,-10.135],[248,-9.9218],[269,-9.9662],[294,-10.135]]
       }
       leg_r_pv_ik_ctrl.Location.Z {
        type=MovieSceneScriptingFloatChannel
        default=0
        thinned_from=22
        keys [[0,70.5794],[25,70.4197],[42,74.1496],[50,75.2612],[76,76.4527],[87,75.7143],[97,73.6987],[111,73.3619],[127,71.4007],[134,69.5522],[142,69.4429],[174,71.2501],[185,71.0824],[197,70.5794],[216,70.4197],[248,70.6206],[294,70.4197]]
       }
       leg_r_pv_ik_ctrl.Rotation.X {
        type=MovieSceneScriptingFloatChannel
        default=0
        thinned_from=22
        keys [[0,-102.7875],[25,-102.9055],[42,-101.2489],[50,-99.4375],[58,-99.1628],[68,-99.1757],[76,-99.8483],[87,-99.0183],[97,-98.6151],[111,-98.9781],[127,-94.9372],[134,-93.4555],[142,-98.0103],[170,-101.2812],[177,-101.3497],[185,-102.5075],[197,-102.7875],[216,-102.9055],[269,-102.7875],[294,-102.9055]]
       }
       leg_r_pv_ik_ctrl.Rotation.Y {
        type=MovieSceneScriptingFloatChannel
        default=0
        thinned_from=22
        keys [[0,71.6085],[25,71.8391],[42,66.479],[50,64.2993],[58,63.548],[68,62.9472],[76,62.746],[87,63.4843],[97,66.2219],[111,66.8227],[127,68.2099],[134,70.407],[142,71.9351],[174,70.3963],[177,70.4449],[197,71.6085],[216,71.8391],[248,71.559],[269,71.6085],[294,71.8391]]
       }
       leg_r_pv_ik_ctrl.Rotation.Z {
        type=MovieSceneScriptingFloatChannel
        default=0
        thinned_from=22
        keys [[0,-111.5892],[25,-112.0072],[42,-110.2013],[50,-107.242],[58,-106.5406],[68,-106.3914],[76,-107.1524],[87,-106.8239],[111,-107.3177],[127,-103.859],[134,-102.1164],[142,-105.0837],[170,-111.2177],[174,-111.0813],[177,-110.7311],[185,-110.3875],[197,-111.5892],[216,-112.0072],[248,-111.4836],[269,-111.5892],[294,-112.0072]]
       }
       leg_r_pv_ik_ctrl.Scale.X {type=MovieSceneScriptingFloatChannel, default=1, const [1,22]}
       leg_r_pv_ik_ctrl.Scale.Y {type=MovieSceneScriptingFloatChannel, default=1, const [1,22]}
       leg_r_pv_ik_ctrl.Scale.Z {type=MovieSceneScriptingFloatChannel, default=1, const [1,22]}
       upperarm_l_fk_ctrl.Location.X {type=MovieSceneScriptingFloatChannel, default=0, thinned_from=29, keys [[0,0.6086],[294,0.6086]]}
       upperarm_l_fk_ctrl.Location.Y {type=MovieSceneScriptingFloatChannel, default=0, thinned_from=29, keys [[0,-2.0367],[294,-2.0367]]}
       upperarm_l_fk_ctrl.Location.Z {type=MovieSceneScriptingFloatChannel, default=0, thinned_from=29, keys [[0,0.0814],[294,0.0814]]}
       upperarm_l_fk_ctrl.Rotation.X {
        type=MovieSceneScriptingFloatChannel
        default=0
        thinned_from=28
        keys [[0,3.5059],[25,3.5059],[32,3.7654],[42,-0.3945],[50,-2.3967],[58,-2.5687],[119,-2.5687],[127,-4.7932],[134,1.7786],[142,8.8603],[159,17.2752],[170,8.3756],[174,6.1436],[177,5.0544],[185,3.5059],[294,3.5059]]
       }
       upperarm_l_fk_ctrl.Rotation.Y {
        type=MovieSceneScriptingFloatChannel
        default=0
        keys [[0,1.3932],[5,1.265],[25,0.8676],[32,0.6714],[42,3.2121],[50,2.6773],[58,2.3036],[68,2.1252],[76,2.1252],[87,1.3357],[97,0.6711],[111,-0.5152],[119,-0.9483],[127,-0.2465],[134,2.0935],[142,0.5904],[149,0.0612],[159,1.4677],[170,1.1336],[174,1.2346],[177,1.3069],[185,1.1336],[197,-0.0321],[216,-0.6613],[236,0.1493],[253,1.1956],[269,1.3932],[274,1.265],[294,0.8676]]
       }
       upperarm_l_fk_ctrl.Rotation.Z {
        type=MovieSceneScriptingFloatChannel
        default=0
        thinned_from=29
        keys [[0,-16.2887],[25,-16.313],[32,-15.6341],[42,-16.7059],[50,-17.4784],[58,-17.7457],[97,-17.7457],[111,-18.0757],[119,-18.2155],[127,-17.6561],[134,-15.3245],[142,-9.8672],[149,-6.2407],[159,-10.9427],[166,-16.655],[170,-17.4192],[173,-17.6773],[185,-16.1576],[197,-16.2764],[216,-16.316],[274,-16.2867],[294,-16.313]]
       }
       upperarm_l_fk_ctrl.Scale.X {type=MovieSceneScriptingFloatChannel, default=1, const [1,4]}
       upperarm_l_fk_ctrl.Scale.Y {type=MovieSceneScriptingFloatChannel, default=1, const [1,4]}
       upperarm_l_fk_ctrl.Scale.Z {type=MovieSceneScriptingFloatChannel, default=1, const [1,4]}
       lowerarm_l_fk_ctrl.Location.X {type=MovieSceneScriptingFloatChannel, default=0, thinned_from=9, keys [[0,0],[294,0]]}
       lowerarm_l_fk_ctrl.Location.Y {type=MovieSceneScriptingFloatChannel, default=0, thinned_from=9, keys [[0,0],[294,0]]}
       lowerarm_l_fk_ctrl.Location.Z {type=MovieSceneScriptingFloatChannel, default=0, thinned_from=9, keys [[0,0],[294,0]]}
       lowerarm_l_fk_ctrl.Rotation.X {
        type=MovieSceneScriptingFloatChannel
        default=0
        thinned_from=26
        keys [[0,-5.2112],[87,-5.2118],[97,-4.9119],[111,-4.8747],[119,-4.8694],[127,-4.9065],[134,-5.1685],[142,-5.153],[149,-4.8378],[159,-5.1166],[170,-5.2112],[294,-5.2112]]
       }
       lowerarm_l_fk_ctrl.Rotation.Y {
        type=MovieSceneScriptingFloatChannel
        default=0
        thinned_from=26
        keys [[0,-2.6616],[25,-2.6616],[32,-2.5577],[42,-1.8307],[50,-1.5103],[58,-1.3955],[127,-1.3955],[134,-2.4803],[142,-2.689],[149,-2.7747],[170,-2.6616],[294,-2.6616]]
       }
       lowerarm_l_fk_ctrl.Rotation.Z {
        type=MovieSceneScriptingFloatChannel
        default=0
        keys [[0,9.4657],[25,10.323],[32,10.9931],[42,10.454],[58,9.2723],[68,8.4541],[76,8.4541],[87,7.2915],[97,5.6345],[111,4.9428],[119,4.844],[127,5.6438],[134,6.9451],[142,10.314],[149,15.7078],[159,18.5591],[170,14.8084],[173,13.3904],[177,12.4004],[185,13.3496],[197,14.8378],[216,12.9733],[248,10.7642],[269,9.4657],[294,10.323]]
       }
       lowerarm_l_fk_ctrl.Scale.X {type=MovieSceneScriptingFloatChannel, default=1, const [1,26]}
       lowerarm_l_fk_ctrl.Scale.Y {type=MovieSceneScriptingFloatChannel, default=1, const [1,26]}
       lowerarm_l_fk_ctrl.Scale.Z {type=MovieSceneScriptingFloatChannel, default=1, const [1,26]}
       hand_l_fk_ctrl.Location.X {type=MovieSceneScriptingFloatChannel, default=0, thinned_from=25, keys [[0,0],[294,0]]}
       hand_l_fk_ctrl.Location.Y {type=MovieSceneScriptingFloatChannel, default=0, thinned_from=25, keys [[0,-0.431],[294,-0.431]]}
       hand_l_fk_ctrl.Location.Z {type=MovieSceneScriptingFloatChannel, default=0, thinned_from=25, keys [[0,0.0668],[294,0.0668]]}
       hand_l_fk_ctrl.Rotation.X {
        type=MovieSceneScriptingFloatChannel
        default=0
        thinned_from=26
        keys [[0,-4.3121],[25,-4.3121],[32,-3.9447],[42,-1.3724],[50,-0.6941],[97,-0.6941],[111,-0.6969],[119,-0.6973],[127,-1.0898],[134,-4.1991],[142,-4.3121],[149,-4.3121],[159,-2.6547],[170,-3.9665],[174,-4.2249],[177,-4.3121],[294,-4.3121]]
       }
       hand_l_fk_ctrl.Rotation.Y {
        type=MovieSceneScriptingFloatChannel
        default=0
        thinned_from=26
        keys [[0,12.0164],[25,11.1688],[32,11.2798],[42,12.0574],[50,12.2624],[119,12.2614],[127,11.7078],[134,8.4589],[142,6.8332],[149,8.5635],[159,15.2455],[170,13.2488],[174,11.8249],[182,8.4208],[190,11.1475],[197,12.0164],[216,11.1688],[248,11.1688],[269,12.0164],[294,11.1688]]
       }
       hand_l_fk_ctrl.Rotation.Z {
        type=MovieSceneScriptingFloatChannel
        default=0
        thinned_from=26
        keys [[0,-1.5901],[25,-1.5901],[32,-1.5109],[42,-0.9567],[50,-0.8105],[119,-0.8112],[127,-0.8958],[134,-1.5657],[142,-1.5901],[149,-1.5901],[159,3.8047],[170,-0.4652],[174,-1.3062],[177,-1.5901],[294,-1.5901]]
       }
       hand_l_fk_ctrl.Scale.X {type=MovieSceneScriptingFloatChannel, default=1, const [1,25]}
       hand_l_fk_ctrl.Scale.Y {type=MovieSceneScriptingFloatChannel, default=1, const [1,25]}
       hand_l_fk_ctrl.Scale.Z {type=MovieSceneScriptingFloatChannel, default=1, const [1,25]}
       thumb_01_l_ctrl.Location.X {type=MovieSceneScriptingFloatChannel, default=0, thinned_from=22, keys [[0,1.5271],[294,1.5271]]}
       thumb_01_l_ctrl.Location.Y {type=MovieSceneScriptingFloatChannel, default=0, thinned_from=22, keys [[0,-0.2099],[294,-0.2099]]}
       thumb_01_l_ctrl.Location.Z {type=MovieSceneScriptingFloatChannel, default=0, thinned_from=22, keys [[0,0.5418],[294,0.5418]]}
       thumb_01_l_ctrl.Rotation.X {type=MovieSceneScriptingFloatChannel, default=0, thinned_from=22, keys [[0,2.1082],[294,2.1082]]}
       thumb_01_l_ctrl.Rotation.Y {type=MovieSceneScriptingFloatChannel, default=0, thinned_from=22, keys [[0,-0.4919],[294,-0.4919]]}
       thumb_01_l_ctrl.Rotation.Z {type=MovieSceneScriptingFloatChannel, default=0, thinned_from=22, keys [[0,1.6614],[294,1.6614]]}
       thumb_01_l_ctrl.Scale.X {type=MovieSceneScriptingFloatChannel, default=1, const [1,22]}
       thumb_01_l_ctrl.Scale.Y {type=MovieSceneScriptingFloatChannel, default=1, const [1,22]}
       thumb_01_l_ctrl.Scale.Z {type=MovieSceneScriptingFloatChannel, default=1, const [1,22]}
       thumb_02_l_ctrl.Location.X {type=MovieSceneScriptingFloatChannel, default=0, thinned_from=22, keys [[0,-0.4653],[294,-0.4653]]}
       thumb_02_l_ctrl.Location.Y {type=MovieSceneScriptingFloatChannel, default=0, thinned_from=22, keys [[0,-1.0209],[294,-1.0209]]}
       thumb_02_l_ctrl.Location.Z {type=MovieSceneScriptingFloatChannel, default=0, thinned_from=22, keys [[0,0.015],[294,0.015]]}
       thumb_02_l_ctrl.Rotation.X {type=MovieSceneScriptingFloatChannel, default=0, const [2.9867,22]}
       thumb_02_l_ctrl.Rotation.Y {type=MovieSceneScriptingFloatChannel, default=0, const [10.4177,22]}
       thumb_02_l_ctrl.Rotation.Z {type=MovieSceneScriptingFloatChannel, default=0, const [-5.5084,22]}
       thumb_02_l_ctrl.Scale.X {type=MovieSceneScriptingFloatChannel, default=1, const [1,22]}
       thumb_02_l_ctrl.Scale.Y {type=MovieSceneScriptingFloatChannel, default=1, const [1,22]}
       thumb_02_l_ctrl.Scale.Z {type=MovieSceneScriptingFloatChannel, default=1, const [1,22]}
       thumb_03_l_ctrl.Location.X {type=MovieSceneScriptingFloatChannel, default=0, const [0,22]}
       thumb_03_l_ctrl.Location.Y {type=MovieSceneScriptingFloatChannel, default=0, const [1,22]}
       thumb_03_l_ctrl.Location.Z {type=MovieSceneScriptingFloatChannel, default=0, thinned_from=22, keys [[0,0.024],[294,0.024]]}
       thumb_03_l_ctrl.Rotation.X {type=MovieSceneScriptingFloatChannel, default=0, thinned_from=22, keys [[0,-0.0739],[294,-0.0739]]}
       thumb_03_l_ctrl.Rotation.Y {type=MovieSceneScriptingFloatChannel, default=0, thinned_from=22, keys [[0,-0.2879],[294,-0.2879]]}
       thumb_03_l_ctrl.Rotation.Z {type=MovieSceneScriptingFloatChannel, default=0, const [-37.1171,22]}
       thumb_03_l_ctrl.Scale.X {type=MovieSceneScriptingFloatChannel, default=1, thinned_from=22, keys [[0,0.9288],[294,0.9288]]}
       thumb_03_l_ctrl.Scale.Y {type=MovieSceneScriptingFloatChannel, default=1, const [1,22]}
       thumb_03_l_ctrl.Scale.Z {type=MovieSceneScriptingFloatChannel, default=1, const [1,22]}
       index_metacarpal_l_ctrl.Location.X {type=MovieSceneScriptingFloatChannel, default=0, thinned_from=22, keys [[0,-0.1076],[294,-0.1076]]}
       index_metacarpal_l_ctrl.Location.Y {type=MovieSceneScriptingFloatChannel, default=0, const [-0.0214,22]}
       index_metacarpal_l_ctrl.Location.Z {type=MovieSceneScriptingFloatChannel, default=0, thinned_from=22, keys [[0,0.0042],[294,0.0042]]}
       index_metacarpal_l_ctrl.Rotation.X {type=MovieSceneScriptingFloatChannel, default=0, const [6.5584,22]}
       index_metacarpal_l_ctrl.Rotation.Y {type=MovieSceneScriptingFloatChannel, default=0, const [-0.6742,22]}
       index_metacarpal_l_ctrl.Rotation.Z {type=MovieSceneScriptingFloatChannel, default=0, const [2.5144,22]}
       index_metacarpal_l_ctrl.Scale.X {type=MovieSceneScriptingFloatChannel, default=1, const [1,22]}
       index_metacarpal_l_ctrl.Scale.Y {type=MovieSceneScriptingFloatChannel, default=1, const [1,22]}
       index_metacarpal_l_ctrl.Scale.Z {type=MovieSceneScriptingFloatChannel, default=1, const [1,22]}
       index_01_l_ctrl.Location.X {type=MovieSceneScriptingFloatChannel, default=0, const [0,22]}
       index_01_l_ctrl.Location.Y {type=MovieSceneScriptingFloatChannel, default=0, const [0,22]}
       index_01_l_ctrl.Location.Z {type=MovieSceneScriptingFloatChannel, default=0, const [0,22]}
       index_01_l_ctrl.Rotation.X {type=MovieSceneScriptingFloatChannel, default=0, const [6.6605,22]}
       index_01_l_ctrl.Rotation.Y {type=MovieSceneScriptingFloatChannel, default=0, const [-18.6454,22]}
       index_01_l_ctrl.Rotation.Z {type=MovieSceneScriptingFloatChannel, default=0, const [12.5523,22]}
       index_01_l_ctrl.Scale.X {type=MovieSceneScriptingFloatChannel, default=1, const [1,22]}
       index_01_l_ctrl.Scale.Y {type=MovieSceneScriptingFloatChannel, default=1, const [1,22]}
       index_01_l_ctrl.Scale.Z {type=MovieSceneScriptingFloatChannel, default=1, const [1,22]}
       index_02_l_ctrl.Location.X {type=MovieSceneScriptingFloatChannel, default=0, const [0,22]}
       index_02_l_ctrl.Location.Y {type=MovieSceneScriptingFloatChannel, default=0, const [0,22]}
       index_02_l_ctrl.Location.Z {type=MovieSceneScriptingFloatChannel, default=0, const [0,22]}
       index_02_l_ctrl.Rotation.X {type=MovieSceneScriptingFloatChannel, default=0, const [0,22]}
       index_02_l_ctrl.Rotation.Y {type=MovieSceneScriptingFloatChannel, default=0, const [0,22]}
       index_02_l_ctrl.Rotation.Z {type=MovieSceneScriptingFloatChannel, default=0, thinned_from=22, keys [[0,0.4999],[294,0.4999]]}
       index_02_l_ctrl.Scale.X {type=MovieSceneScriptingFloatChannel, default=1, const [1,22]}
       index_02_l_ctrl.Scale.Y {type=MovieSceneScriptingFloatChannel, default=1, const [1,22]}
       index_02_l_ctrl.Scale.Z {type=MovieSceneScriptingFloatChannel, default=1, const [1,22]}
       index_03_l_ctrl.Location.X {type=MovieSceneScriptingFloatChannel, default=0, const [0,22]}
       index_03_l_ctrl.Location.Y {type=MovieSceneScriptingFloatChannel, default=0, const [0,22]}
       index_03_l_ctrl.Location.Z {type=MovieSceneScriptingFloatChannel, default=0, const [0,22]}
       index_03_l_ctrl.Rotation.X {type=MovieSceneScriptingFloatChannel, default=0, const [0,22]}
       index_03_l_ctrl.Rotation.Y {type=MovieSceneScriptingFloatChannel, default=0, const [0,22]}
       index_03_l_ctrl.Rotation.Z {type=MovieSceneScriptingFloatChannel, default=0, thinned_from=22, keys [[0,0.4999],[294,0.4999]]}
       index_03_l_ctrl.Scale.X {type=MovieSceneScriptingFloatChannel, default=1, const [1,22]}
       index_03_l_ctrl.Scale.Y {type=MovieSceneScriptingFloatChannel, default=1, const [1,22]}
       index_03_l_ctrl.Scale.Z {type=MovieSceneScriptingFloatChannel, default=1, const [1,22]}
       middle_metacarpal_l_ctrl.Location.X {type=MovieSceneScriptingFloatChannel, default=0, const [0,22]}
       middle_metacarpal_l_ctrl.Location.Y {type=MovieSceneScriptingFloatChannel, default=0, const [0,22]}
       middle_metacarpal_l_ctrl.Location.Z {type=MovieSceneScriptingFloatChannel, default=0, const [0,22]}
       middle_metacarpal_l_ctrl.Rotation.X {type=MovieSceneScriptingFloatChannel, default=0, const [12.2183,22]}
       middle_metacarpal_l_ctrl.Rotation.Y {type=MovieSceneScriptingFloatChannel, default=0, const [0.8305,22]}
       middle_metacarpal_l_ctrl.Rotation.Z {type=MovieSceneScriptingFloatChannel, default=0, const [-3.3667,22]}
       middle_metacarpal_l_ctrl.Scale.X {type=MovieSceneScriptingFloatChannel, default=1, const [1,22]}
       middle_metacarpal_l_ctrl.Scale.Y {type=MovieSceneScriptingFloatChannel, default=1, const [1,22]}
       middle_metacarpal_l_ctrl.Scale.Z {type=MovieSceneScriptingFloatChannel, default=1, const [1,22]}
       middle_01_l_ctrl.Location.X {type=MovieSceneScriptingFloatChannel, default=0, const [0,22]}
       middle_01_l_ctrl.Location.Y {type=MovieSceneScriptingFloatChannel, default=0, const [0,22]}
       middle_01_l_ctrl.Location.Z {type=MovieSceneScriptingFloatChannel, default=0, const [0,22]}
       middle_01_l_ctrl.Rotation.X {type=MovieSceneScriptingFloatChannel, default=0, const [4.0257,22]}
       middle_01_l_ctrl.Rotation.Y {type=MovieSceneScriptingFloatChannel, default=0, thinned_from=22, keys [[0,0.7654],[294,0.7654]]}
       middle_01_l_ctrl.Rotation.Z {type=MovieSceneScriptingFloatChannel, default=0, thinned_from=22, keys [[0,-0.6452],[294,-0.6452]]}
       middle_01_l_ctrl.Scale.X {type=MovieSceneScriptingFloatChannel, default=1, const [1,22]}
       middle_01_l_ctrl.Scale.Y {type=MovieSceneScriptingFloatChannel, default=1, const [1,22]}
       middle_01_l_ctrl.Scale.Z {type=MovieSceneScriptingFloatChannel, default=1, const [1,22]}
       middle_02_l_ctrl.Location.X {type=MovieSceneScriptingFloatChannel, default=0, const [0,22]}
       middle_02_l_ctrl.Location.Y {type=MovieSceneScriptingFloatChannel, default=0, const [0,22]}
       middle_02_l_ctrl.Location.Z {type=MovieSceneScriptingFloatChannel, default=0, const [0,22]}
       middle_02_l_ctrl.Rotation.X {type=MovieSceneScriptingFloatChannel, default=0, const [0,22]}
       middle_02_l_ctrl.Rotation.Y {type=MovieSceneScriptingFloatChannel, default=0, const [0,22]}
       middle_02_l_ctrl.Rotation.Z {type=MovieSceneScriptingFloatChannel, default=0, thinned_from=22, keys [[0,0.4999],[294,0.4999]]}
       middle_02_l_ctrl.Scale.X {type=MovieSceneScriptingFloatChannel, default=1, const [1,22]}
       middle_02_l_ctrl.Scale.Y {type=MovieSceneScriptingFloatChannel, default=1, const [1,22]}
       middle_02_l_ctrl.Scale.Z {type=MovieSceneScriptingFloatChannel, default=1, const [1,22]}
       middle_03_l_ctrl.Location.X {type=MovieSceneScriptingFloatChannel, default=0, const [0,22]}
       middle_03_l_ctrl.Location.Y {type=MovieSceneScriptingFloatChannel, default=0, const [0,22]}
       middle_03_l_ctrl.Location.Z {type=MovieSceneScriptingFloatChannel, default=0, const [0,22]}
       middle_03_l_ctrl.Rotation.X {type=MovieSceneScriptingFloatChannel, default=0, const [0,22]}
       middle_03_l_ctrl.Rotation.Y {type=MovieSceneScriptingFloatChannel, default=0, const [0,22]}
       middle_03_l_ctrl.Rotation.Z {type=MovieSceneScriptingFloatChannel, default=0, thinned_from=22, keys [[0,0.4999],[294,0.4999]]}
       middle_03_l_ctrl.Scale.X {type=MovieSceneScriptingFloatChannel, default=1, const [1,22]}
       middle_03_l_ctrl.Scale.Y {type=MovieSceneScriptingFloatChannel, default=1, const [1,22]}
       middle_03_l_ctrl.Scale.Z {type=MovieSceneScriptingFloatChannel, default=1, const [1,22]}
       ring_metacarpal_l_ctrl.Location.X {type=MovieSceneScriptingFloatChannel, default=0, const [0,22]}
       ring_metacarpal_l_ctrl.Location.Y {type=MovieSceneScriptingFloatChannel, default=0, const [0,22]}
       ring_metacarpal_l_ctrl.Location.Z {type=MovieSceneScriptingFloatChannel, default=0, const [0,22]}
       ring_metacarpal_l_ctrl.Rotation.X {type=MovieSceneScriptingFloatChannel, default=0, thinned_from=22, keys [[0,-0.1822],[294,-0.1822]]}
       ring_metacarpal_l_ctrl.Rotation.Y {type=MovieSceneScriptingFloatChannel, default=0, const [3.8684,22]}
       ring_metacarpal_l_ctrl.Rotation.Z {type=MovieSceneScriptingFloatChannel, default=0, const [3.0752,22]}
       ring_metacarpal_l_ctrl.Scale.X {type=MovieSceneScriptingFloatChannel, default=1, const [1,22]}
       ring_metacarpal_l_ctrl.Scale.Y {type=MovieSceneScriptingFloatChannel, default=1, const [1,22]}
       ring_metacarpal_l_ctrl.Scale.Z {type=MovieSceneScriptingFloatChannel, default=1, const [1,22]}
       ring_01_l_ctrl.Location.X {type=MovieSceneScriptingFloatChannel, default=0, const [0,22]}
       ring_01_l_ctrl.Location.Y {type=MovieSceneScriptingFloatChannel, default=0, const [0,22]}
       ring_01_l_ctrl.Location.Z {type=MovieSceneScriptingFloatChannel, default=0, const [0,22]}
       ring_01_l_ctrl.Rotation.X {type=MovieSceneScriptingFloatChannel, default=0, const [19.1792,22]}
       ring_01_l_ctrl.Rotation.Y {type=MovieSceneScriptingFloatChannel, default=0, thinned_from=22, keys [[0,0.8469],[294,0.8469]]}
       ring_01_l_ctrl.Rotation.Z {type=MovieSceneScriptingFloatChannel, default=0, const [-7.065,22]}
       ring_01_l_ctrl.Scale.X {type=MovieSceneScriptingFloatChannel, default=1, const [1,22]}
       ring_01_l_ctrl.Scale.Y {type=MovieSceneScriptingFloatChannel, default=1, const [1,22]}
       ring_01_l_ctrl.Scale.Z {type=MovieSceneScriptingFloatChannel, default=1, const [1,22]}
       ring_02_l_ctrl.Location.X {type=MovieSceneScriptingFloatChannel, default=0, const [0,22]}
       ring_02_l_ctrl.Location.Y {type=MovieSceneScriptingFloatChannel, default=0, const [0,22]}
       ring_02_l_ctrl.Location.Z {type=MovieSceneScriptingFloatChannel, default=0, const [0,22]}
       ring_02_l_ctrl.Rotation.X {type=MovieSceneScriptingFloatChannel, default=0, const [0,22]}
       ring_02_l_ctrl.Rotation.Y {type=MovieSceneScriptingFloatChannel, default=0, const [0,22]}
       ring_02_l_ctrl.Rotation.Z {type=MovieSceneScriptingFloatChannel, default=0, const [29.8737,22]}
       ring_02_l_ctrl.Scale.X {type=MovieSceneScriptingFloatChannel, default=1, const [1,22]}
       ring_02_l_ctrl.Scale.Y {type=MovieSceneScriptingFloatChannel, default=1, const [1,22]}
       ring_02_l_ctrl.Scale.Z {type=MovieSceneScriptingFloatChannel, default=1, const [1,22]}
       ring_03_l_ctrl.Location.X {type=MovieSceneScriptingFloatChannel, default=0, const [0,22]}
       ring_03_l_ctrl.Location.Y {type=MovieSceneScriptingFloatChannel, default=0, const [0,22]}
       ring_03_l_ctrl.Location.Z {type=MovieSceneScriptingFloatChannel, default=0, const [0,22]}
       ring_03_l_ctrl.Rotation.X {type=MovieSceneScriptingFloatChannel, default=0, const [0,22]}
       ring_03_l_ctrl.Rotation.Y {type=MovieSceneScriptingFloatChannel, default=0, const [0,22]}
       ring_03_l_ctrl.Rotation.Z {type=MovieSceneScriptingFloatChannel, default=0, const [-33.2338,22]}
       ring_03_l_ctrl.Scale.X {type=MovieSceneScriptingFloatChannel, default=1, const [1,22]}
       ring_03_l_ctrl.Scale.Y {type=MovieSceneScriptingFloatChannel, default=1, const [1,22]}
       ring_03_l_ctrl.Scale.Z {type=MovieSceneScriptingFloatChannel, default=1, const [1,22]}
       pinky_metacarpal_l_ctrl.Location.X {type=MovieSceneScriptingFloatChannel, default=0, const [0,22]}
       pinky_metacarpal_l_ctrl.Location.Y {type=MovieSceneScriptingFloatChannel, default=0, const [0,22]}
       pinky_metacarpal_l_ctrl.Location.Z {type=MovieSceneScriptingFloatChannel, default=0, const [0,22]}
       pinky_metacarpal_l_ctrl.Rotation.X {type=MovieSceneScriptingFloatChannel, default=0, const [-10.0612,22]}
       pinky_metacarpal_l_ctrl.Rotation.Y {type=MovieSceneScriptingFloatChannel, default=0, const [5.1753,22]}
       pinky_metacarpal_l_ctrl.Rotation.Z {type=MovieSceneScriptingFloatChannel, default=0, const [-3.5711,22]}
       pinky_metacarpal_l_ctrl.Scale.X {type=MovieSceneScriptingFloatChannel, default=1, const [1,22]}
       pinky_metacarpal_l_ctrl.Scale.Y {type=MovieSceneScriptingFloatChannel, default=1, const [1,22]}
       pinky_metacarpal_l_ctrl.Scale.Z {type=MovieSceneScriptingFloatChannel, default=1, const [1,22]}
       pinky_01_l_ctrl.Location.X {type=MovieSceneScriptingFloatChannel, default=0, const [0,22]}
       pinky_01_l_ctrl.Location.Y {type=MovieSceneScriptingFloatChannel, default=0, const [0,22]}
       pinky_01_l_ctrl.Location.Z {type=MovieSceneScriptingFloatChannel, default=0, const [0,22]}
       pinky_01_l_ctrl.Rotation.X {type=MovieSceneScriptingFloatChannel, default=0, const [8.2602,22]}
       pinky_01_l_ctrl.Rotation.Y {type=MovieSceneScriptingFloatChannel, default=0, const [9.0512,22]}
       pinky_01_l_ctrl.Rotation.Z {type=MovieSceneScriptingFloatChannel, default=0, const [21.5373,22]}
       pinky_01_l_ctrl.Scale.X {type=MovieSceneScriptingFloatChannel, default=1, const [1,22]}
       pinky_01_l_ctrl.Scale.Y {type=MovieSceneScriptingFloatChannel, default=1, const [1,22]}
       pinky_01_l_ctrl.Scale.Z {type=MovieSceneScriptingFloatChannel, default=1, const [1,22]}
       pinky_02_l_ctrl.Location.X {type=MovieSceneScriptingFloatChannel, default=0, const [0,22]}
       pinky_02_l_ctrl.Location.Y {type=MovieSceneScriptingFloatChannel, default=0, const [0,22]}
       pinky_02_l_ctrl.Location.Z {type=MovieSceneScriptingFloatChannel, default=0, const [0,22]}
       pinky_02_l_ctrl.Rotation.X {type=MovieSceneScriptingFloatChannel, default=0, thinned_from=22, keys [[0,-0.0138],[294,-0.0138]]}
       pinky_02_l_ctrl.Rotation.Y {type=MovieSceneScriptingFloatChannel, default=0, thinned_from=22, keys [[0,0.023],[294,0.023]]}
       pinky_02_l_ctrl.Rotation.Z {type=MovieSceneScriptingFloatChannel, default=0, const [3.9655,22]}
       pinky_02_l_ctrl.Scale.X {type=MovieSceneScriptingFloatChannel, default=1, const [1,22]}
       pinky_02_l_ctrl.Scale.Y {type=MovieSceneScriptingFloatChannel, default=1, const [1,22]}
       pinky_02_l_ctrl.Scale.Z {type=MovieSceneScriptingFloatChannel, default=1, const [1,22]}
       pinky_03_l_ctrl.Location.X {type=MovieSceneScriptingFloatChannel, default=0, const [0,22]}
       pinky_03_l_ctrl.Location.Y {type=MovieSceneScriptingFloatChannel, default=0, const [0,22]}
       pinky_03_l_ctrl.Location.Z {type=MovieSceneScriptingFloatChannel, default=0, const [0,22]}
       pinky_03_l_ctrl.Rotation.X {type=MovieSceneScriptingFloatChannel, default=0, const [1.0695,22]}
       pinky_03_l_ctrl.Rotation.Y {type=MovieSceneScriptingFloatChannel, default=0, const [0.5593,22]}
       pinky_03_l_ctrl.Rotation.Z {type=MovieSceneScriptingFloatChannel, default=0, const [-50.5377,22]}
       pinky_03_l_ctrl.Scale.X {type=MovieSceneScriptingFloatChannel, default=1, const [1,22]}
       pinky_03_l_ctrl.Scale.Y {type=MovieSceneScriptingFloatChannel, default=1, const [1,22]}
       pinky_03_l_ctrl.Scale.Z {type=MovieSceneScriptingFloatChannel, default=1, const [1,22]}
       hand_l_ik_ctrl.Location.X {
        type=MovieSceneScriptingFloatChannel
        default=0
        keys [[0,5.2583],[25,5.2275],[32,6.0101],[42,5.0043],[50,4.9997],[58,4.9119],[68,4.4046],[76,4.9119],[87,5.6471],[97,9.3155],[111,8.8606],[119,8.7956],[127,8.4708],[134,5.0674],[142,4.681],[149,3.1287],[169,6.479],[173,5.0239],[176,5.8338],[184,4.7413],[197,4.8148],[216,4.9089],[248,4.6994],[269,5.2572],[294,5.2275]]
       }
       hand_l_ik_ctrl.Location.Y {
        type=MovieSceneScriptingFloatChannel
        default=0
        thinned_from=25
        keys [[0,17.2668],[25,17.9735],[32,13.3094],[42,12.1396],[50,10.3444],[58,10.3275],[68,10.8674],[76,10.3275],[87,9.4301],[97,7.2381],[111,7.1419],[119,7.1282],[127,7.4669],[134,12.2599],[142,13.1297],[149,13.1683],[169,13.1544],[173,18.0719],[176,12.7798],[184,18.6691],[197,18.4265],[248,18.63],[269,17.2645],[294,17.9735]]
       }
       hand_l_ik_ctrl.Location.Z {
        type=MovieSceneScriptingFloatChannel
        default=0
        keys [[0,13.12],[25,13.1231],[32,17.2443],[42,14.9862],[50,13.4098],[58,13.3468],[68,12.5961],[76,13.3468],[87,13.5875],[97,12.9866],[111,11.464],[119,11.2465],[127,11.9665],[134,19.9103],[142,20.8178],[149,19.6979],[169,17.2117],[173,12.0303],[176,16.5426],[184,14.1733],[197,13.917],[216,13.857],[248,14.1994],[269,13.1187],[294,13.1231]]
       }
       hand_l_ik_ctrl.Rotation.X {
        type=MovieSceneScriptingFloatChannel
        default=0
        thinned_from=25
        keys [[0,30.8727],[25,30.4811],[32,35.6533],[42,31.5676],[50,25.3812],[58,23.4989],[68,25.2625],[76,23.4989],[87,25.3724],[97,18.0914],[119,18.0694],[127,19.5332],[134,36.3171],[142,38.4542],[149,34.0597],[169,36.0268],[173,31.4591],[176,35.2976],[184,30.9068],[197,30.7887],[216,30.7806],[248,31.0001],[269,30.8727],[294,30.4811]]
       }
       hand_l_ik_ctrl.Rotation.Y {
        type=MovieSceneScriptingFloatChannel
        default=0
        keys [[0,9.8302],[25,9.8049],[32,14.8137],[42,16.5084],[50,15.7877],[58,16.1066],[68,15.9797],[76,16.1066],[87,15.8636],[97,11.6753],[111,12.7328],[119,12.8839],[127,12.9571],[134,13.4599],[142,13.4551],[149,17.7397],[169,16.2098],[173,10.1056],[176,14.1228],[184,12.6203],[197,11.9236],[216,11.7986],[248,13.1135],[269,9.8302],[294,9.8049]]
       }
       hand_l_ik_ctrl.Rotation.Z {
        type=MovieSceneScriptingFloatChannel
        default=0
        keys [[0,29.807],[25,30.26],[32,27.0883],[42,22.7807],[50,16.6954],[58,15.3216],[68,17.8299],[76,15.3216],[87,17.2731],[97,11.22],[111,12.6361],[119,12.8384],[127,13.6364],[134,20.7162],[142,21.1957],[149,20.1811],[169,29.5528],[173,33.8163],[176,26.06],[184,32.6516],[197,32.0465],[216,31.9879],[248,33.0166],[269,29.807],[294,30.26]]
       }
       hand_l_ik_ctrl.Scale.X {type=MovieSceneScriptingFloatChannel, default=1, const [1,25]}
       hand_l_ik_ctrl.Scale.Y {type=MovieSceneScriptingFloatChannel, default=1, const [1,25]}
       hand_l_ik_ctrl.Scale.Z {type=MovieSceneScriptingFloatChannel, default=1, const [1,25]}
       Softness {type=MovieSceneScriptingFloatChannel, default=0, const [0,25]}
       hand_l_ikorient_ctrl.Rotation.Roll {type=MovieSceneScriptingFloatChannel, default=0, const [0,22]}
       hand_l_ikorient_ctrl.Rotation.Pitch {type=MovieSceneScriptingFloatChannel, default=0, const [0,22]}
       hand_l_ikorient_ctrl.Rotation.Yaw {type=MovieSceneScriptingFloatChannel, default=0, const [0,22]}
       arm_l_pv_ik_ctrl.Location.X {
        type=MovieSceneScriptingFloatChannel
        default=0
        keys [[0,6.0688],[25,5.7334],[32,9.1226],[42,6.8504],[50,6.7574],[58,6.0551],[68,4.4585],[76,6.0551],[87,8.0412],[97,7.31],[111,5.9753],[119,5.7847],[127,6.2275],[134,7.6202],[142,6.9246],[149,5.9723],[169,9.5891],[173,7.1197],[176,9.2857],[184,5.7838],[197,5.7592],[216,5.7696],[248,5.7636],[269,6.0709],[294,5.7334]]
       }
       arm_l_pv_ik_ctrl.Location.Y {
        type=MovieSceneScriptingFloatChannel
        default=0
        thinned_from=25
        keys [[0,21.4479],[25,21.2139],[32,23.8817],[42,18.7389],[50,14.461],[58,13.6137],[68,14.5386],[76,13.6137],[87,14.9476],[97,15.187],[119,15.2449],[127,15.978],[134,25.0415],[142,26.2943],[149,21.7516],[169,23.937],[173,20.573],[176,23.1823],[184,20.9873],[197,21.162],[248,21.1584],[269,21.4463],[294,21.2139]]
       }
       arm_l_pv_ik_ctrl.Location.Z {
        type=MovieSceneScriptingFloatChannel
        default=0
        keys [[0,2.1121],[25,1.6433],[32,4.1503],[42,4.0767],[50,5.1039],[58,5.0614],[68,4.317],[76,5.0614],[87,4.8209],[97,2.117],[111,0.3625],[119,0.1118],[127,0.9641],[134,9.4118],[142,10.2043],[149,9.6027],[169,3.4621],[173,1.8683],[176,4.7594],[184,2.1149],[197,1.9024],[216,1.6432],[248,1.8997],[269,2.1134],[294,1.6433]]
       }
       arm_l_pv_ik_ctrl.Rotation.X {
        type=MovieSceneScriptingFloatChannel
        default=0
        keys [[0,15.1377],[25,14.8766],[32,15.5442],[42,15.0222],[50,16.7537],[58,16.6053],[68,15.5904],[76,16.6053],[87,16.5657],[97,14.1515],[111,12.1088],[119,11.817],[127,12.6415],[134,20.5881],[142,21.3236],[149,19.9167],[169,14.9351],[173,16.632],[176,16.5601],[184,15.8615],[197,15.0996],[216,14.8455],[248,15.0982],[269,15.1377],[294,14.8766]]
       }
       arm_l_pv_ik_ctrl.Rotation.Y {
        type=MovieSceneScriptingFloatChannel
        default=0
        keys [[0,-3.5759],[25,-4.0897],[32,3.7241],[42,3.9285],[50,1.925],[58,1.6602],[68,3.5471],[76,1.6602],[87,1.3178],[97,-1.0315],[111,1.6343],[119,2.0151],[127,2.2972],[134,7.8616],[142,9.0367],[149,9.4169],[169,3.846],[173,-5.616],[176,4.2041],[184,-3.9781],[197,-4.0793],[216,-4.1137],[248,-4.078],[269,-3.5759],[294,-4.0897]]
       }
       arm_l_pv_ik_ctrl.Rotation.Z {
        type=MovieSceneScriptingFloatChannel
        default=0
        thinned_from=25
        keys [[0,41.2305],[25,41.0715],[32,42.2418],[42,35.0215],[50,29.0795],[58,27.3932],[68,27.7269],[76,27.3932],[87,29.3911],[97,23.3075],[111,21.9404],[119,21.7451],[127,23.4168],[134,39.8832],[142,41.4456],[149,36.9192],[169,42.2765],[176,41.9673],[184,40.6792],[197,40.9664],[216,41.0635],[248,40.9674],[269,41.2305],[294,41.0715]]
       }
       arm_l_pv_ik_ctrl.Scale.X {type=MovieSceneScriptingFloatChannel, default=1, const [1,25]}
       arm_l_pv_ik_ctrl.Scale.Y {type=MovieSceneScriptingFloatChannel, default=1, const [1,25]}
       arm_l_pv_ik_ctrl.Scale.Z {type=MovieSceneScriptingFloatChannel, default=1, const [1,25]}
       upperarm_r_fk_ctrl.Location.X {
        type=MovieSceneScriptingFloatChannel
        default=0
        thinned_from=31
        keys [[0,-0.3332],[119,-0.3329],[127,-0.259],[134,-0.1039],[139,-0.0373],[155,-0.0373],[159,-0.0931],[166,-0.2945],[169,-0.3332],[294,-0.3332]]
       }
       upperarm_r_fk_ctrl.Location.Y {
        type=MovieSceneScriptingFloatChannel
        default=0
        thinned_from=31
        keys [[0,2.1155],[119,2.1137],[127,1.6445],[134,0.6595],[139,0.2369],[155,0.2369],[159,0.5912],[166,1.8699],[169,2.1155],[294,2.1155]]
       }
       upperarm_r_fk_ctrl.Location.Z {
        type=MovieSceneScriptingFloatChannel
        default=0
        thinned_from=31
        keys [[0,0.612],[119,0.6115],[127,0.4758],[134,0.1908],[139,0.0685],[155,0.0685],[159,0.171],[166,0.541],[169,0.612],[294,0.612]]
       }
       upperarm_r_fk_ctrl.Rotation.X {
        type=MovieSceneScriptingFloatChannel
        default=0
        thinned_from=27
        keys [[0,3.651],[32,3.6535],[42,4.7684],[50,8.2454],[58,9.4154],[76,9.4154],[87,8.6211],[97,5.4783],[111,1.8492],[119,2.6799],[127,4.4986],[134,2.3058],[139,-0.3508],[150,-2.883],[159,-1.7203],[166,2.8263],[184,6.9131],[197,6.3559],[216,4.0982],[236,3.651],[294,3.651]]
       }
       upperarm_r_fk_ctrl.Rotation.Y {
        type=MovieSceneScriptingFloatChannel
        default=0
        keys [[0,-6.6721],[5,-6.8236],[25,-7.2552],[32,-7.6044],[42,-7.376],[50,-5.6839],[58,-5.1384],[68,-4.9662],[76,-4.9488],[87,-5.2116],[97,-6.2516],[111,-7.7755],[119,-8.2601],[127,-7.8163],[134,-7.1156],[139,-6.4848],[150,-4.2058],[155,-3.7423],[159,-4.3643],[166,-5.8681],[184,-7.9187],[197,-8.5946],[216,-8.9414],[236,-8.1542],[253,-6.9333],[269,-6.6721],[274,-6.8236],[294,-7.2552]]
       }
       upperarm_r_fk_ctrl.Rotation.Z {
        type=MovieSceneScriptingFloatChannel
        default=0
        thinned_from=30
        keys [[0,-12.468],[5,-12.4741],[25,-12.2206],[32,-12.5183],[42,-14.333],[50,-13.5593],[58,-12.4794],[68,-11.7139],[76,-10.6503],[87,-10.1379],[97,-12.2354],[111,-14.1719],[119,-14.8072],[127,-13.582],[134,-8.795],[139,-6.3561],[150,-8.8857],[155,-10.9848],[159,-12.4571],[166,-13.2255],[173,-11.6541],[176,-11.2418],[184,-11.071],[197,-11.9026],[216,-12.1912],[269,-12.468],[274,-12.4741],[294,-12.2206]]
       }
       upperarm_r_fk_ctrl.Scale.X {type=MovieSceneScriptingFloatChannel, default=1, const [1,6]}
       upperarm_r_fk_ctrl.Scale.Y {type=MovieSceneScriptingFloatChannel, default=1, const [1,6]}
       upperarm_r_fk_ctrl.Scale.Z {type=MovieSceneScriptingFloatChannel, default=1, const [1,6]}
       lowerarm_r_fk_ctrl.Location.X {type=MovieSceneScriptingFloatChannel, default=0, thinned_from=27, keys [[0,0],[294,0]]}
       lowerarm_r_fk_ctrl.Location.Y {type=MovieSceneScriptingFloatChannel, default=0, thinned_from=27, keys [[0,0],[294,0]]}
       lowerarm_r_fk_ctrl.Location.Z {type=MovieSceneScriptingFloatChannel, default=0, thinned_from=27, keys [[0,0],[294,0]]}
       lowerarm_r_fk_ctrl.Rotation.X {
        type=MovieSceneScriptingFloatChannel
        default=0
        thinned_from=28
        keys [[0,-7.1742],[119,-7.1701],[127,-7.1416],[134,-7.335],[139,-7.4516],[150,-7.2434],[155,-7.1785],[294,-7.1743]]
       }
       lowerarm_r_fk_ctrl.Rotation.Y {type=MovieSceneScriptingFloatChannel, default=0, thinned_from=24, keys [[0,-2.926],[294,-2.926]]}
       lowerarm_r_fk_ctrl.Rotation.Z {
        type=MovieSceneScriptingFloatChannel
        default=0
        keys [[0,10.4332],[25,9.4506],[32,8.57],[42,7.146],[50,7.3917],[58,7.9323],[68,9.1064],[76,10.3103],[87,12.6345],[97,12.8607],[111,10.821],[119,10.9369],[127,11.7487],[134,12.8628],[142,13.212],[150,10.6255],[155,7.726],[159,5.4793],[166,4.7599],[169,6.3623],[173,8.5532],[184,11.364],[197,8.7956],[216,6.7475],[248,9.58],[269,10.4332],[294,9.4506]]
       }
       lowerarm_r_fk_ctrl.Scale.X {type=MovieSceneScriptingFloatChannel, default=1, const [1,27]}
       lowerarm_r_fk_ctrl.Scale.Y {type=MovieSceneScriptingFloatChannel, default=1, const [1,27]}
       lowerarm_r_fk_ctrl.Scale.Z {type=MovieSceneScriptingFloatChannel, default=1, const [1,27]}
       hand_r_fk_ctrl.Location.X {type=MovieSceneScriptingFloatChannel, default=0, thinned_from=20, keys [[0,0],[294,0]]}
       hand_r_fk_ctrl.Location.Y {type=MovieSceneScriptingFloatChannel, default=0.431, thinned_from=20, keys [[0,0.431],[294,0.431]]}
       hand_r_fk_ctrl.Location.Z {type=MovieSceneScriptingFloatChannel, default=-0.0668, thinned_from=20, keys [[0,-0.0668],[294,-0.0668]]}
       hand_r_fk_ctrl.Rotation.X {
        type=MovieSceneScriptingFloatChannel
        default=0
        thinned_from=24
        keys [[0,-26.6562],[25,-26.6445],[32,-26.5934],[42,-25.8616],[50,-25.3735],[68,-25.4772],[87,-25.7041],[97,-25.0704],[111,-22.8784],[119,-21.4409],[127,-20.7098],[134,-20.9148],[142,-21.4531],[169,-26.6562],[294,-26.6445]]
       }
       hand_r_fk_ctrl.Rotation.Y {
        type=MovieSceneScriptingFloatChannel
        default=0
        keys [[0,11.4956],[25,10.7034],[32,10.5161],[42,8.8215],[50,6.8123],[58,5.3211],[68,4.0766],[76,3.3311],[87,2.8589],[97,3.7485],[111,5.2237],[119,4.3283],[127,2.0623],[134,2.6468],[150,10.7034],[169,5.4277],[184,9.8467],[197,10.7034],[216,7.2863],[248,10.3747],[269,11.4956],[294,10.7034]]
       }
       hand_r_fk_ctrl.Rotation.Z {
        type=MovieSceneScriptingFloatChannel
        default=0
        keys [[0,-8.2564],[25,-7.1752],[32,-6.8314],[42,-2.2079],[50,0.9024],[58,0.7473],[68,0.2879],[76,-0.2087],[87,-1.0566],[97,-1.9313],[111,-4.825],[119,-4.2047],[127,-3.0099],[134,-3.3582],[142,-5.5288],[166,-9.9243],[184,-7.18],[197,-6.3972],[216,-4.914],[248,-7.1752],[269,-8.2564],[294,-7.1752]]
       }
       hand_r_fk_ctrl.Scale.X {type=MovieSceneScriptingFloatChannel, default=1, const [1,24]}
       hand_r_fk_ctrl.Scale.Y {type=MovieSceneScriptingFloatChannel, default=1, const [1,24]}
       hand_r_fk_ctrl.Scale.Z {type=MovieSceneScriptingFloatChannel, default=1, const [1,24]}
       thumb_01_r_ctrl.Location.X {type=MovieSceneScriptingFloatChannel, default=0, const [0,22]}
       thumb_01_r_ctrl.Location.Y {type=MovieSceneScriptingFloatChannel, default=0, const [0,22]}
       thumb_01_r_ctrl.Location.Z {type=MovieSceneScriptingFloatChannel, default=0, const [0,22]}
       thumb_01_r_ctrl.Rotation.X {type=MovieSceneScriptingFloatChannel, default=0, thinned_from=22, keys [[0,11.4575],[97,11.4575],[111,9.099],[127,11.4575],[294,11.4575]]}
       thumb_01_r_ctrl.Rotation.Y {type=MovieSceneScriptingFloatChannel, default=0, const [9.7162,22]}
       thumb_01_r_ctrl.Rotation.Z {type=MovieSceneScriptingFloatChannel, default=0, thinned_from=22, keys [[0,-7.0005],[294,-7.0005]]}
       thumb_01_r_ctrl.Scale.X {type=MovieSceneScriptingFloatChannel, default=1, const [1,22]}
       thumb_01_r_ctrl.Scale.Y {type=MovieSceneScriptingFloatChannel, default=1, const [1,22]}
       thumb_01_r_ctrl.Scale.Z {type=MovieSceneScriptingFloatChannel, default=1, const [1,22]}
       thumb_02_r_ctrl.Location.X {type=MovieSceneScriptingFloatChannel, default=0, thinned_from=22, keys [[0,-0.011],[294,-0.011]]}
       thumb_02_r_ctrl.Location.Y {type=MovieSceneScriptingFloatChannel, default=0, thinned_from=22, keys [[0,-0.4402],[294,-0.4402]]}
       thumb_02_r_ctrl.Location.Z {type=MovieSceneScriptingFloatChannel, default=0, const [0,22]}
       thumb_02_r_ctrl.Rotation.X {type=MovieSceneScriptingFloatChannel, default=0, thinned_from=22, keys [[0,0],[294,0]]}
       thumb_02_r_ctrl.Rotation.Y {type=MovieSceneScriptingFloatChannel, default=0, thinned_from=22, keys [[0,0],[294,0]]}
       thumb_02_r_ctrl.Rotation.Z {type=MovieSceneScriptingFloatChannel, default=0, const [7.2573,22]}
       thumb_02_r_ctrl.Scale.X {type=MovieSceneScriptingFloatChannel, default=1, const [1,22]}
       thumb_02_r_ctrl.Scale.Y {type=MovieSceneScriptingFloatChannel, default=1, const [1,22]}
       thumb_02_r_ctrl.Scale.Z {type=MovieSceneScriptingFloatChannel, default=1, const [1,22]}
       thumb_03_r_ctrl.Location.X {type=MovieSceneScriptingFloatChannel, default=0, const [0,22]}
       thumb_03_r_ctrl.Location.Y {type=MovieSceneScriptingFloatChannel, default=0, const [0,22]}
       thumb_03_r_ctrl.Location.Z {type=MovieSceneScriptingFloatChannel, default=0, const [0,22]}
       thumb_03_r_ctrl.Rotation.X {type=MovieSceneScriptingFloatChannel, default=0, thinned_from=22, keys [[0,0],[294,0]]}
       thumb_03_r_ctrl.Rotation.Y {type=MovieSceneScriptingFloatChannel, default=0, thinned_from=22, keys [[0,0],[294,0]]}
       thumb_03_r_ctrl.Rotation.Z {type=MovieSceneScriptingFloatChannel, default=0, const [-37.5339,22]}
       thumb_03_r_ctrl.Scale.X {type=MovieSceneScriptingFloatChannel, default=1, const [1,22]}
       thumb_03_r_ctrl.Scale.Y {type=MovieSceneScriptingFloatChannel, default=1, const [1,22]}
       thumb_03_r_ctrl.Scale.Z {type=MovieSceneScriptingFloatChannel, default=1, const [1,22]}
       index_metacarpal_r_ctrl.Location.X {type=MovieSceneScriptingFloatChannel, default=0, const [0,22]}
       index_metacarpal_r_ctrl.Location.Y {type=MovieSceneScriptingFloatChannel, default=0, const [0,22]}
       index_metacarpal_r_ctrl.Location.Z {type=MovieSceneScriptingFloatChannel, default=0, const [0,22]}
       index_metacarpal_r_ctrl.Rotation.X {type=MovieSceneScriptingFloatChannel, default=0, const [21.7659,22]}
       index_metacarpal_r_ctrl.Rotation.Y {type=MovieSceneScriptingFloatChannel, default=0, const [-5.906,22]}
       index_metacarpal_r_ctrl.Rotation.Z {type=MovieSceneScriptingFloatChannel, default=0, const [12.4289,22]}
       index_metacarpal_r_ctrl.Scale.X {type=MovieSceneScriptingFloatChannel, default=1, const [1,22]}
       index_metacarpal_r_ctrl.Scale.Y {type=MovieSceneScriptingFloatChannel, default=1, const [1,22]}
       index_metacarpal_r_ctrl.Scale.Z {type=MovieSceneScriptingFloatChannel, default=1, const [1,22]}
       index_01_r_ctrl.Location.X {type=MovieSceneScriptingFloatChannel, default=0, const [0,22]}
       index_01_r_ctrl.Location.Y {type=MovieSceneScriptingFloatChannel, default=0, const [0,22]}
       index_01_r_ctrl.Location.Z {type=MovieSceneScriptingFloatChannel, default=0, const [0,22]}
       index_01_r_ctrl.Rotation.X {type=MovieSceneScriptingFloatChannel, default=0, const [6.372,22]}
       index_01_r_ctrl.Rotation.Y {type=MovieSceneScriptingFloatChannel, default=0, const [-9.2464,22]}
       index_01_r_ctrl.Rotation.Z {type=MovieSceneScriptingFloatChannel, default=0, const [-1.9369,22]}
       index_01_r_ctrl.Scale.X {type=MovieSceneScriptingFloatChannel, default=1, const [1,22]}
       index_01_r_ctrl.Scale.Y {type=MovieSceneScriptingFloatChannel, default=1, const [1,22]}
       index_01_r_ctrl.Scale.Z {type=MovieSceneScriptingFloatChannel, default=1, const [1,22]}
       index_02_r_ctrl.Location.X {type=MovieSceneScriptingFloatChannel, default=0, const [0,22]}
       index_02_r_ctrl.Location.Y {type=MovieSceneScriptingFloatChannel, default=0, const [0,22]}
       index_02_r_ctrl.Location.Z {type=MovieSceneScriptingFloatChannel, default=0, const [0,22]}
       index_02_r_ctrl.Rotation.X {type=MovieSceneScriptingFloatChannel, default=0, const [0,22]}
       index_02_r_ctrl.Rotation.Y {type=MovieSceneScriptingFloatChannel, default=0, const [0,22]}
       index_02_r_ctrl.Rotation.Z {type=MovieSceneScriptingFloatChannel, default=0, thinned_from=22, keys [[0,0.4999],[294,0.4999]]}
       index_02_r_ctrl.Scale.X {type=MovieSceneScriptingFloatChannel, default=1, const [1,22]}
       index_02_r_ctrl.Scale.Y {type=MovieSceneScriptingFloatChannel, default=1, const [1,22]}
       index_02_r_ctrl.Scale.Z {type=MovieSceneScriptingFloatChannel, default=1, const [1,22]}
       index_03_r_ctrl.Location.X {type=MovieSceneScriptingFloatChannel, default=0, const [0,22]}
       index_03_r_ctrl.Location.Y {type=MovieSceneScriptingFloatChannel, default=0, const [0,22]}
       index_03_r_ctrl.Location.Z {type=MovieSceneScriptingFloatChannel, default=0, const [0,22]}
       index_03_r_ctrl.Rotation.X {type=MovieSceneScriptingFloatChannel, default=0, const [0,22]}
       index_03_r_ctrl.Rotation.Y {type=MovieSceneScriptingFloatChannel, default=0, const [0,22]}
       index_03_r_ctrl.Rotation.Z {type=MovieSceneScriptingFloatChannel, default=0, thinned_from=22, keys [[0,0.4999],[294,0.4999]]}
       index_03_r_ctrl.Scale.X {type=MovieSceneScriptingFloatChannel, default=1, const [1,22]}
       index_03_r_ctrl.Scale.Y {type=MovieSceneScriptingFloatChannel, default=1, const [1,22]}
       index_03_r_ctrl.Scale.Z {type=MovieSceneScriptingFloatChannel, default=1, const [1,22]}
       middle_metacarpal_r_ctrl.Location.X {type=MovieSceneScriptingFloatChannel, default=0, const [0,22]}
       middle_metacarpal_r_ctrl.Location.Y {type=MovieSceneScriptingFloatChannel, default=0, const [0,22]}
       middle_metacarpal_r_ctrl.Location.Z {type=MovieSceneScriptingFloatChannel, default=0, const [0,22]}
       middle_metacarpal_r_ctrl.Rotation.X {type=MovieSceneScriptingFloatChannel, default=0, thinned_from=22, keys [[0,7.9652],[294,7.9652]]}
       middle_metacarpal_r_ctrl.Rotation.Y {type=MovieSceneScriptingFloatChannel, default=0, thinned_from=22, keys [[0,-0.1426],[294,-0.1426]]}
       middle_metacarpal_r_ctrl.Rotation.Z {type=MovieSceneScriptingFloatChannel, default=0, thinned_from=22, keys [[0,7.5327],[294,7.5327]]}
       middle_metacarpal_r_ctrl.Scale.X {type=MovieSceneScriptingFloatChannel, default=1, const [1,22]}
       middle_metacarpal_r_ctrl.Scale.Y {type=MovieSceneScriptingFloatChannel, default=1, const [1,22]}
       middle_metacarpal_r_ctrl.Scale.Z {type=MovieSceneScriptingFloatChannel, default=1, const [1,22]}
       middle_01_r_ctrl.Location.X {type=MovieSceneScriptingFloatChannel, default=0, const [0,22]}
       middle_01_r_ctrl.Location.Y {type=MovieSceneScriptingFloatChannel, default=0, const [0,22]}
       middle_01_r_ctrl.Location.Z {type=MovieSceneScriptingFloatChannel, default=0, const [0,22]}
       middle_01_r_ctrl.Rotation.X {type=MovieSceneScriptingFloatChannel, default=0, const [12.6674,22]}
       middle_01_r_ctrl.Rotation.Y {type=MovieSceneScriptingFloatChannel, default=0, thinned_from=22, keys [[0,-0.3057],[294,-0.3057]]}
       middle_01_r_ctrl.Rotation.Z {type=MovieSceneScriptingFloatChannel, default=0, const [-5.985,22]}
       middle_01_r_ctrl.Scale.X {type=MovieSceneScriptingFloatChannel, default=1, const [1,22]}
       middle_01_r_ctrl.Scale.Y {type=MovieSceneScriptingFloatChannel, default=1, const [1,22]}
       middle_01_r_ctrl.Scale.Z {type=MovieSceneScriptingFloatChannel, default=1, const [1,22]}
       middle_02_r_ctrl.Location.X {type=MovieSceneScriptingFloatChannel, default=0, const [0,22]}
       middle_02_r_ctrl.Location.Y {type=MovieSceneScriptingFloatChannel, default=0, const [0,22]}
       middle_02_r_ctrl.Location.Z {type=MovieSceneScriptingFloatChannel, default=0, const [0,22]}
       middle_02_r_ctrl.Rotation.X {type=MovieSceneScriptingFloatChannel, default=0, const [0,22]}
       middle_02_r_ctrl.Rotation.Y {type=MovieSceneScriptingFloatChannel, default=0, const [0,22]}
       middle_02_r_ctrl.Rotation.Z {type=MovieSceneScriptingFloatChannel, default=0, thinned_from=22, keys [[0,0.4999],[294,0.4999]]}
       middle_02_r_ctrl.Scale.X {type=MovieSceneScriptingFloatChannel, default=1, const [1,22]}
       middle_02_r_ctrl.Scale.Y {type=MovieSceneScriptingFloatChannel, default=1, const [1,22]}
       middle_02_r_ctrl.Scale.Z {type=MovieSceneScriptingFloatChannel, default=1, const [1,22]}
       middle_03_r_ctrl.Location.X {type=MovieSceneScriptingFloatChannel, default=0, const [0,22]}
       middle_03_r_ctrl.Location.Y {type=MovieSceneScriptingFloatChannel, default=0, const [0,22]}
       middle_03_r_ctrl.Location.Z {type=MovieSceneScriptingFloatChannel, default=0, const [0,22]}
       middle_03_r_ctrl.Rotation.X {type=MovieSceneScriptingFloatChannel, default=0, const [0,22]}
       middle_03_r_ctrl.Rotation.Y {type=MovieSceneScriptingFloatChannel, default=0, const [0,22]}
       middle_03_r_ctrl.Rotation.Z {type=MovieSceneScriptingFloatChannel, default=0, thinned_from=22, keys [[0,0.4999],[294,0.4999]]}
       middle_03_r_ctrl.Scale.X {type=MovieSceneScriptingFloatChannel, default=1, const [1,22]}
       middle_03_r_ctrl.Scale.Y {type=MovieSceneScriptingFloatChannel, default=1, const [1,22]}
       middle_03_r_ctrl.Scale.Z {type=MovieSceneScriptingFloatChannel, default=1, const [1,22]}
       ring_metacarpal_r_ctrl.Location.X {type=MovieSceneScriptingFloatChannel, default=0, const [0,22]}
       ring_metacarpal_r_ctrl.Location.Y {type=MovieSceneScriptingFloatChannel, default=0, const [0,22]}
       ring_metacarpal_r_ctrl.Location.Z {type=MovieSceneScriptingFloatChannel, default=0, const [0,22]}
       ring_metacarpal_r_ctrl.Rotation.X {type=MovieSceneScriptingFloatChannel, default=0, const [7.9728,22]}
       ring_metacarpal_r_ctrl.Rotation.Y {type=MovieSceneScriptingFloatChannel, default=0, thinned_from=22, keys [[0,2.4941],[294,2.4941]]}
       ring_metacarpal_r_ctrl.Rotation.Z {type=MovieSceneScriptingFloatChannel, default=0, const [7.9018,22]}
       ring_metacarpal_r_ctrl.Scale.X {type=MovieSceneScriptingFloatChannel, default=1, const [1,22]}
       ring_metacarpal_r_ctrl.Scale.Y {type=MovieSceneScriptingFloatChannel, default=1, const [1,22]}
       ring_metacarpal_r_ctrl.Scale.Z {type=MovieSceneScriptingFloatChannel, default=1, const [1,22]}
       ring_01_r_ctrl.Location.X {type=MovieSceneScriptingFloatChannel, default=0, const [0,22]}
       ring_01_r_ctrl.Location.Y {type=MovieSceneScriptingFloatChannel, default=0, const [0,22]}
       ring_01_r_ctrl.Location.Z {type=MovieSceneScriptingFloatChannel, default=0, const [0,22]}
       ring_01_r_ctrl.Rotation.X {type=MovieSceneScriptingFloatChannel, default=0, const [2.0002,22]}
       ring_01_r_ctrl.Rotation.Y {type=MovieSceneScriptingFloatChannel, default=0, const [4.8346,22]}
       ring_01_r_ctrl.Rotation.Z {type=MovieSceneScriptingFloatChannel, default=0, const [-16.5444,22]}
       ring_01_r_ctrl.Scale.X {type=MovieSceneScriptingFloatChannel, default=1, const [1,22]}
       ring_01_r_ctrl.Scale.Y {type=MovieSceneScriptingFloatChannel, default=1, const [1,22]}
       ring_01_r_ctrl.Scale.Z {type=MovieSceneScriptingFloatChannel, default=1, const [1,22]}
       ring_02_r_ctrl.Location.X {type=MovieSceneScriptingFloatChannel, default=0, const [0,22]}
       ring_02_r_ctrl.Location.Y {type=MovieSceneScriptingFloatChannel, default=0, const [0,22]}
       ring_02_r_ctrl.Location.Z {type=MovieSceneScriptingFloatChannel, default=0, const [0,22]}
       ring_02_r_ctrl.Rotation.X {type=MovieSceneScriptingFloatChannel, default=0, const [0,22]}
       ring_02_r_ctrl.Rotation.Y {type=MovieSceneScriptingFloatChannel, default=0, const [0,22]}
       ring_02_r_ctrl.Rotation.Z {type=MovieSceneScriptingFloatChannel, default=0, const [16.9398,22]}
       ring_02_r_ctrl.Scale.X {type=MovieSceneScriptingFloatChannel, default=1, const [1,22]}
       ring_02_r_ctrl.Scale.Y {type=MovieSceneScriptingFloatChannel, default=1, const [1,22]}
       ring_02_r_ctrl.Scale.Z {type=MovieSceneScriptingFloatChannel, default=1, const [1,22]}
       ring_03_r_ctrl.Location.X {type=MovieSceneScriptingFloatChannel, default=0, const [0,22]}
       ring_03_r_ctrl.Location.Y {type=MovieSceneScriptingFloatChannel, default=0, const [0,22]}
       ring_03_r_ctrl.Location.Z {type=MovieSceneScriptingFloatChannel, default=0, const [0,22]}
       ring_03_r_ctrl.Rotation.X {type=MovieSceneScriptingFloatChannel, default=0, const [0,22]}
       ring_03_r_ctrl.Rotation.Y {type=MovieSceneScriptingFloatChannel, default=0, const [0,22]}
       ring_03_r_ctrl.Rotation.Z {type=MovieSceneScriptingFloatChannel, default=0, const [-33.2338,22]}
       ring_03_r_ctrl.Scale.X {type=MovieSceneScriptingFloatChannel, default=1, const [1,22]}
       ring_03_r_ctrl.Scale.Y {type=MovieSceneScriptingFloatChannel, default=1, const [1,22]}
       ring_03_r_ctrl.Scale.Z {type=MovieSceneScriptingFloatChannel, default=1, const [1,22]}
       pinky_metacarpal_r_ctrl.Location.X {type=MovieSceneScriptingFloatChannel, default=0, const [0,22]}
       pinky_metacarpal_r_ctrl.Location.Y {type=MovieSceneScriptingFloatChannel, default=0, const [0,22]}
       pinky_metacarpal_r_ctrl.Location.Z {type=MovieSceneScriptingFloatChannel, default=0, const [0,22]}
       pinky_metacarpal_r_ctrl.Rotation.X {type=MovieSceneScriptingFloatChannel, default=0, const [-11.2885,22]}
       pinky_metacarpal_r_ctrl.Rotation.Y {type=MovieSceneScriptingFloatChannel, default=0, thinned_from=22, keys [[0,4.5525],[294,4.5525]]}
       pinky_metacarpal_r_ctrl.Rotation.Z {type=MovieSceneScriptingFloatChannel, default=0, const [-5.1966,22]}
       pinky_metacarpal_r_ctrl.Scale.X {type=MovieSceneScriptingFloatChannel, default=1, const [1,22]}
       pinky_metacarpal_r_ctrl.Scale.Y {type=MovieSceneScriptingFloatChannel, default=1, const [1,22]}
       pinky_metacarpal_r_ctrl.Scale.Z {type=MovieSceneScriptingFloatChannel, default=1, const [1,22]}
       pinky_01_r_ctrl.Location.X {type=MovieSceneScriptingFloatChannel, default=0, const [0,22]}
       pinky_01_r_ctrl.Location.Y {type=MovieSceneScriptingFloatChannel, default=0, const [0,22]}
       pinky_01_r_ctrl.Location.Z {type=MovieSceneScriptingFloatChannel, default=0, const [0,22]}
       pinky_01_r_ctrl.Rotation.X {type=MovieSceneScriptingFloatChannel, default=0, const [17.3586,22]}
       pinky_01_r_ctrl.Rotation.Y {type=MovieSceneScriptingFloatChannel, default=0, const [8.0471,22]}
       pinky_01_r_ctrl.Rotation.Z {type=MovieSceneScriptingFloatChannel, default=0, const [23.3265,22]}
       pinky_01_r_ctrl.Scale.X {type=MovieSceneScriptingFloatChannel, default=1, const [1,22]}
       pinky_01_r_ctrl.Scale.Y {type=MovieSceneScriptingFloatChannel, default=1, const [1,22]}
       pinky_01_r_ctrl.Scale.Z {type=MovieSceneScriptingFloatChannel, default=1, const [1,22]}
       pinky_02_r_ctrl.Location.X {type=MovieSceneScriptingFloatChannel, default=0, const [0,22]}
       pinky_02_r_ctrl.Location.Y {type=MovieSceneScriptingFloatChannel, default=0, const [0,22]}
       pinky_02_r_ctrl.Location.Z {type=MovieSceneScriptingFloatChannel, default=0, const [0,22]}
       pinky_02_r_ctrl.Rotation.X {type=MovieSceneScriptingFloatChannel, default=0, thinned_from=22, keys [[0,-0.0138],[294,-0.0138]]}
       pinky_02_r_ctrl.Rotation.Y {type=MovieSceneScriptingFloatChannel, default=0, thinned_from=22, keys [[0,0.023],[294,0.023]]}
       pinky_02_r_ctrl.Rotation.Z {type=MovieSceneScriptingFloatChannel, default=0, const [3.9655,22]}
       pinky_02_r_ctrl.Scale.X {type=MovieSceneScriptingFloatChannel, default=1, const [1,22]}
       pinky_02_r_ctrl.Scale.Y {type=MovieSceneScriptingFloatChannel, default=1, const [1,22]}
       pinky_02_r_ctrl.Scale.Z {type=MovieSceneScriptingFloatChannel, default=1, const [1,22]}
       pinky_03_r_ctrl.Location.X {type=MovieSceneScriptingFloatChannel, default=0, const [0,22]}
       pinky_03_r_ctrl.Location.Y {type=MovieSceneScriptingFloatChannel, default=0, const [0,22]}
       pinky_03_r_ctrl.Location.Z {type=MovieSceneScriptingFloatChannel, default=0, const [0,22]}
       pinky_03_r_ctrl.Rotation.X {type=MovieSceneScriptingFloatChannel, default=0, const [1.0695,22]}
       pinky_03_r_ctrl.Rotation.Y {type=MovieSceneScriptingFloatChannel, default=0, const [0.5593,22]}
       pinky_03_r_ctrl.Rotation.Z {type=MovieSceneScriptingFloatChannel, default=0, const [-50.5377,22]}
       pinky_03_r_ctrl.Scale.X {type=MovieSceneScriptingFloatChannel, default=1, const [1,22]}
       pinky_03_r_ctrl.Scale.Y {type=MovieSceneScriptingFloatChannel, default=1, const [1,22]}
       pinky_03_r_ctrl.Scale.Z {type=MovieSceneScriptingFloatChannel, default=1, const [1,22]}
       hand_r_ik_ctrl.Location.X {
        type=MovieSceneScriptingFloatChannel
        default=0
        thinned_from=24
        keys [[0,-6.9841],[25,-6.5573],[32,-6.148],[42,-4.3311],[50,-2.5867],[58,-1.5811],[68,-0.6675],[76,-0.0828],[87,-1.6802],[97,-4.4525],[111,-6.1045],[119,-6.34],[127,-10.6171],[134,-3.0416],[142,-9.194],[170,-11.2483],[174,-6.4663],[177,-11.1717],[185,-6.9949],[197,-6.6639],[248,-6.2878],[269,-6.9859],[294,-6.5573]]
       }
       hand_r_ik_ctrl.Location.Y {
        type=MovieSceneScriptingFloatChannel
        default=0
        keys [[0,7.5145],[25,7.9922],[32,8.2573],[42,6.6069],[50,3.6862],[58,2.4324],[68,1.4162],[76,1.0008],[87,0.9408],[97,3.3309],[111,5.6622],[119,6.6605],[127,2.2018],[134,4.3067],[142,4.1889],[169,3.2229],[173,4.9885],[176,2.9453],[184,7.375],[197,7.7503],[216,7.9531],[248,7.2596],[269,7.5146],[294,7.9922]]
       }
       hand_r_ik_ctrl.Location.Z {
        type=MovieSceneScriptingFloatChannel
        default=0
        thinned_from=24
        keys [[0,-23.905],[25,-23.842],[32,-23.9414],[42,-23.7544],[50,-22.0956],[58,-20.7388],[68,-19.2224],[76,-17.7095],[87,-15.8894],[111,-18.9535],[119,-19.7129],[127,-25.1442],[134,-21.7059],[142,-30.4703],[170,-29.9825],[174,-22.2016],[177,-30.2645],[185,-23.7982],[197,-23.728],[216,-23.746],[248,-23.2884],[269,-23.9048],[294,-23.842]]
       }
       hand_r_ik_ctrl.Rotation.X {
        type=MovieSceneScriptingFloatChannel
        default=0
        keys [[0,-66.1645],[25,-67.2021],[32,-67.8091],[42,-64.8533],[50,-59.8101],[58,-57.6875],[68,-56.7196],[76,-56.1538],[87,-55.9739],[97,-58.3577],[111,-59.7268],[119,-58.8784],[127,-56.3766],[134,-63.2491],[142,-69.5913],[170,-61.2378],[174,-65.8268],[177,-61.2189],[185,-66.3102],[197,-66.9847],[216,-67.2095],[248,-66.8297],[269,-66.1644],[294,-67.2021]]
       }
       hand_r_ik_ctrl.Rotation.Y {
        type=MovieSceneScriptingFloatChannel
        default=0
        thinned_from=24
        keys [[0,-13.2503],[25,-13.3738],[32,-13.0365],[42,-12.3477],[50,-10.3488],[58,-10.0796],[68,-9.6133],[76,-8.4925],[87,-6.9229],[97,-10.7817],[111,-14.7473],[119,-16.5585],[127,-21.4962],[134,-9.065],[142,-24.2892],[169,-21.3954],[173,-11.7258],[176,-21.6567],[184,-14.2584],[197,-13.3541],[248,-12.5463],[269,-13.2483],[294,-13.3738]]
       }
       hand_r_ik_ctrl.Rotation.Z {
        type=MovieSceneScriptingFloatChannel
        default=0
        keys [[0,-21.4065],[25,-19.4565],[32,-18.406],[42,-8.6522],[50,-3.1476],[58,-2.3989],[68,-2.7859],[76,-3.7602],[87,-7.3099],[97,-10.4264],[111,-12.3353],[119,-11.2288],[127,-24.0573],[134,-15.5202],[142,-24.6872],[170,-27.3572],[174,-22.4437],[177,-26.5462],[185,-18.4157],[197,-19.3455],[216,-19.6683],[248,-19.8465],[269,-21.4078],[294,-19.4565]]
       }
       hand_r_ik_ctrl.Scale.X {type=MovieSceneScriptingFloatChannel, default=1, const [1,24]}
       hand_r_ik_ctrl.Scale.Y {type=MovieSceneScriptingFloatChannel, default=1, const [1,24]}
       hand_r_ik_ctrl.Scale.Z {type=MovieSceneScriptingFloatChannel, default=1, const [1,24]}
       Softness_2 {type=MovieSceneScriptingFloatChannel, default=0, const [0,24]}
       hand_r_ikorient_ctrl.Rotation.Roll {type=MovieSceneScriptingFloatChannel, default=0, const [0,22]}
       hand_r_ikorient_ctrl.Rotation.Pitch {type=MovieSceneScriptingFloatChannel, default=0, thinned_from=22, keys [[0,0],[294,0]]}
       hand_r_ikorient_ctrl.Rotation.Yaw {type=MovieSceneScriptingFloatChannel, default=0, const [0,22]}
       arm_r_pv_ik_ctrl.Location.X {
        type=MovieSceneScriptingFloatChannel
        default=0
        keys [[0,-25.9467],[25,-26.0685],[32,-25.9072],[42,-20.6465],[50,-14.9164],[58,-12.7187],[68,-11.5326],[76,-11.2183],[87,-13.1957],[97,-18.7091],[111,-22.892],[119,-23.4695],[127,-22.5044],[134,-21.5566],[142,-27.9565],[170,-25.198],[174,-24.1808],[177,-24.6137],[185,-25.1588],[197,-25.8241],[216,-26.1028],[248,-25.2973],[269,-25.9478],[294,-26.0685]]
       }
       arm_r_pv_ik_ctrl.Location.Y {
        type=MovieSceneScriptingFloatChannel
        default=0
        keys [[0,-16.2772],[25,-15.9469],[32,-15.9719],[42,-15.7497],[50,-15.1761],[58,-14.4388],[68,-13.4388],[76,-12.5971],[87,-12.4242],[97,-12.6136],[111,-12.513],[119,-13.1184],[127,-15.8401],[134,-14.9106],[142,-17.2273],[170,-19.521],[174,-14.9378],[177,-19.7261],[185,-15.9638],[197,-15.9435],[216,-15.9593],[248,-15.6922],[269,-16.2786],[294,-15.9469]]
       }
       arm_r_pv_ik_ctrl.Location.Z {
        type=MovieSceneScriptingFloatChannel
        default=0
        keys [[0,-1.6217],[25,-2.2638],[32,-2.8516],[42,-1.5086],[50,-0.7104],[58,-0.8132],[68,-1.0301],[76,-1.4258],[87,-0.1709],[97,1.2983],[111,1.6073],[119,1.6161],[127,5.6504],[134,-3.7848],[142,-0.151],[170,1.3579],[174,-1.686],[177,1.7333],[185,-0.8022],[197,-1.8891],[216,-2.3199],[248,-2.3363],[269,-1.6205],[294,-2.2638]]
       }
       arm_r_pv_ik_ctrl.Rotation.X {
        type=MovieSceneScriptingFloatChannel
        default=0
        keys [[0,10.7916],[25,10.2754],[32,9.9352],[42,13.5927],[50,14.7479],[58,14.4139],[68,13.9135],[76,13.0037],[87,12.327],[97,12.9679],[111,13.2209],[119,13.95],[127,16.3768],[134,10.5804],[142,13.1973],[170,11.6686],[174,8.8755],[177,12.4834],[185,11.7814],[197,10.6061],[216,10.2146],[248,9.759],[269,10.7916],[294,10.2754]]
       }
       arm_r_pv_ik_ctrl.Rotation.Y {
        type=MovieSceneScriptingFloatChannel
        default=0
        keys [[0,-2.0664],[25,-2.6578],[32,-3.4587],[42,-3.7079],[50,-3.062],[58,-2.8609],[68,-2.6873],[76,-2.5861],[87,-1.1566],[97,0.2407],[111,0.143],[119,-0.7368],[127,9.1325],[134,-5.6943],[142,1.0425],[170,8.1726],[174,-0.9447],[177,7.9282],[185,-2.2666],[197,-2.5802],[216,-2.6855],[248,-2.6132],[269,-2.0664],[294,-2.6578]]
       }
       arm_r_pv_ik_ctrl.Rotation.Z {
        type=MovieSceneScriptingFloatChannel
        default=0
        keys [[0,-46.122],[25,-46.0893],[32,-45.8676],[42,-40.7774],[50,-34.4664],[58,-31.9648],[68,-30.934],[76,-30.2933],[87,-31.2801],[97,-36.5277],[111,-41.4622],[119,-41.6575],[127,-44.7987],[134,-44.186],[142,-56.1601],[169,-50.2931],[173,-45.7083],[176,-50.0143],[184,-45.5384],[197,-45.9425],[216,-46.0759],[248,-45.4872],[269,-46.122],[294,-46.0893]]
       }
       arm_r_pv_ik_ctrl.Scale.X {type=MovieSceneScriptingFloatChannel, default=1, const [1,24]}
       arm_r_pv_ik_ctrl.Scale.Y {type=MovieSceneScriptingFloatChannel, default=1, const [1,24]}
       arm_r_pv_ik_ctrl.Scale.Z {type=MovieSceneScriptingFloatChannel, default=1, const [1,24]}
       body_ctrl.Location.X {
        type=MovieSceneScriptingFloatChannel
        default=-3.4405
        thinned_from=28
        keys [[0,-5.0317],[25,-5.0426],[32,-5.1375],[42,-6.2926],[50,-7.5282],[68,-8.0048],[76,-7.7367],[87,-6.7921],[97,-5.1287],[111,-4.449],[119,-4.6464],[127,-6.2037],[134,-8.1212],[142,-7.7533],[149,-6.3935],[159,-3.9784],[166,-4.1769],[167,-4.3127],[174,-5.5062],[184,-5.5086],[197,-5.1509],[216,-5.0317],[236,-5.0237],[294,-5.0426]]
       }
       body_ctrl.Location.Y {
        type=MovieSceneScriptingFloatChannel
        default=-0.1968
        thinned_from=27
        keys [[0,-2.4997],[25,-2.4997],[32,-2.6684],[42,-2.5239],[50,-1.8442],[76,-0.4907],[87,0.1039],[97,0.6556],[111,0.8083],[119,0.6218],[127,0.0045],[134,-1.6947],[142,-3.3337],[149,-3.8174],[159,-3.2605],[166,-2.4734],[173,-2.1132],[184,-2.3421],[197,-2.4603],[216,-2.4997],[294,-2.4997]]
       }
       body_ctrl.Location.Z {
        type=MovieSceneScriptingFloatChannel
        default=-5.2243
        keys [[0,-2.785],[5,-2.821],[25,-2.8369],[32,-2.918],[42,-2.8608],[50,-2.6757],[68,-2.5646],[76,-2.6042],[87,-2.5646],[97,-2.5827],[111,-2.6284],[119,-2.6887],[127,-2.8165],[134,-2.9405],[142,-2.8932],[149,-3.1002],[159,-3.3032],[166,-2.7776],[173,-2.6024],[184,-2.7704],[197,-2.8281],[216,-2.8199],[236,-2.6577],[253,-2.5936],[269,-2.785],[274,-2.821],[294,-2.8369]]
       }
       body_ctrl.Rotation.X {
        type=MovieSceneScriptingFloatChannel
        default=-12.8141
        keys [[0,-1.5881],[5,-1.5653],[25,-1.8502],[32,-2.4624],[42,-0.8149],[50,-0.197],[68,0.1737],[76,0.2218],[87,0.2218],[97,0.2073],[111,0.2053],[119,0.3277],[127,1.1885],[134,2.1603],[142,2.4843],[149,1.2126],[159,-2.3267],[166,-2.7402],[173,-2.0999],[184,-0.2951],[197,-0.6807],[216,-1.8768],[236,-2.0985],[253,-1.9205],[269,-1.5881],[274,-1.5653],[294,-1.8502]]
       }
       body_ctrl.Rotation.Y {
        type=MovieSceneScriptingFloatChannel
        default=-0.5903
        thinned_from=27
        keys [[0,-4.7599],[25,-4.7599],[32,-4.3174],[42,-3.9476],[50,-3.9015],[68,-3.8738],[76,-3.9169],[87,-5.3882],[97,-7.036],[111,-7.3839],[119,-6.9099],[127,-4.7757],[134,-1.0147],[142,1.6225],[149,0.8025],[159,-2.8602],[166,-5.5058],[173,-5.3229],[184,-4.943],[197,-4.8057],[216,-4.7599],[294,-4.7599]]
       }
       body_ctrl.Rotation.Z {
        type=MovieSceneScriptingFloatChannel
        default=40.1998
        thinned_from=27
        keys [[0,43.5499],[25,43.5499],[32,43.199],[42,38.0652],[50,35.5692],[68,34.0715],[76,33.9247],[87,34.3791],[97,35.5954],[111,35.983],[119,36.296],[127,37.6074],[134,40.3625],[142,42.3335],[149,43.0881],[159,43.3919],[166,42.7194],[173,42.4952],[184,43.0122],[197,43.4155],[216,43.5499],[294,43.5499]]
       }
       body_ctrl.Scale.X {type=MovieSceneScriptingFloatChannel, default=1, const [1,27]}
       body_ctrl.Scale.Y {type=MovieSceneScriptingFloatChannel, default=1, const [1,27]}
       body_ctrl.Scale.Z {type=MovieSceneScriptingFloatChannel, default=1, const [1,27]}
       hips_ctrl.Location.X {type=MovieSceneScriptingFloatChannel, default=0, const [1.5817,27]}
       hips_ctrl.Location.Y {type=MovieSceneScriptingFloatChannel, default=0, const [1.6906,27]}
       hips_ctrl.Location.Z {type=MovieSceneScriptingFloatChannel, default=0, thinned_from=27, keys [[0,-0.0046],[294,-0.0046]]}
       hips_ctrl.Rotation.X {
        type=MovieSceneScriptingFloatChannel
        default=0
        thinned_from=27
        keys [[0,-8.2648],[5,-8.3131],[25,-8.3813],[32,-8.3898],[42,-8.0987],[50,-6.451],[68,-5.4623],[76,-5.5296],[87,-6.0002],[97,-7.1265],[111,-7.2874],[119,-7.4323],[127,-8.5079],[134,-9.6578],[142,-10.0411],[149,-9.7814],[159,-9.105],[166,-8.265],[173,-7.9796],[182,-8.0832],[197,-8.4749],[216,-8.269],[236,-8.1131],[253,-8.0788],[274,-8.3131],[294,-8.3813]]
       }
       hips_ctrl.Rotation.Y {
        type=MovieSceneScriptingFloatChannel
        default=0
        thinned_from=27
        keys [[0,-3.2757],[5,-3.2859],[32,-3.3069],[68,-3.3069],[76,-3.2524],[87,-2.8705],[97,-2.7581],[111,-2.742],[119,-3.104],[127,-5.7906],[134,-8.663],[142,-9.6205],[149,-9.1888],[159,-6.4482],[166,-3.5875],[173,-2.3342],[182,-2.0818],[197,-2.338],[216,-2.8336],[236,-3.1078],[253,-3.2012],[269,-3.2757],[274,-3.2859],[294,-3.3]]
       }
       hips_ctrl.Rotation.Z {
        type=MovieSceneScriptingFloatChannel
        default=0
        thinned_from=27
        keys [[0,-0.0286],[25,-0.0286],[32,-0.0213],[68,-0.0213],[76,-0.0259],[87,0.1972],[97,1.7719],[111,2.0334],[119,1.9938],[127,1.6999],[134,1.3857],[142,1.281],[149,1.281],[159,1.024],[166,-0.0213],[173,-0.5203],[182,-0.354],[197,-0.1461],[216,-0.0286],[294,-0.0286]]
       }
       hips_ctrl.Scale.X {type=MovieSceneScriptingFloatChannel, default=1, const [1,27]}
       hips_ctrl.Scale.Y {type=MovieSceneScriptingFloatChannel, default=1, const [1,27]}
       hips_ctrl.Scale.Z {type=MovieSceneScriptingFloatChannel, default=1, const [1,27]}
       hips_tan_ctrl.Location.X {
        type=MovieSceneScriptingFloatChannel
        default=0
        thinned_from=27
        keys [[0,-3.3763],[5,-3.359],[25,-3.3807],[32,-3.3821],[42,-3.4284],[50,-3.4435],[68,-3.4434],[76,-3.4348],[87,-3.3742],[97,-3.4572],[111,-3.4688],[119,-3.5193],[127,-3.8893],[134,-4.2753],[142,-4.4016],[149,-4.4016],[159,-4.1529],[166,-3.3821],[173,-3.2897],[184,-3.2897],[197,-3.359],[216,-3.3807],[236,-3.3484],[269,-3.3763],[274,-3.359],[294,-3.3807]]
       }
       hips_tan_ctrl.Location.Y {
        type=MovieSceneScriptingFloatChannel
        default=0.008
        thinned_from=27
        keys [[0,-1.682],[25,-1.6623],[32,-1.6613],[42,-1.6957],[50,-2.0519],[68,-2.1094],[76,-2.0987],[87,-2.0233],[97,-1.7491],[111,-1.7097],[119,-1.6928],[127,-1.5705],[134,-1.4466],[142,-1.4071],[149,-1.4071],[159,-1.4538],[166,-1.6613],[173,-1.7313],[184,-1.7313],[197,-1.6788],[216,-1.6623],[236,-1.7094],[253,-1.712],[269,-1.682],[294,-1.6623]]
       }
       hips_tan_ctrl.Location.Z {
        type=MovieSceneScriptingFloatChannel
        default=-0.0073
        keys [[0,0.1961],[5,0.2095],[25,0.2047],[32,0.2046],[42,0.188],[50,0.1022],[68,0.0874],[76,0.0934],[87,0.1351],[97,0.1675],[111,0.1716],[119,0.1555],[127,0.0262],[134,-0.1306],[142,-0.187],[149,-0.187],[159,-0.0755],[166,0.2046],[173,0.2235],[184,0.2235],[197,0.2095],[216,0.2049],[236,0.1672],[253,0.1851],[269,0.1961],[274,0.2095],[294,0.2047]]
       }
       hips_tan_ctrl.Rotation.X {type=MovieSceneScriptingFloatChannel, default=0, const [0,27]}
       hips_tan_ctrl.Rotation.Y {type=MovieSceneScriptingFloatChannel, default=0, const [0,27]}
       hips_tan_ctrl.Rotation.Z {type=MovieSceneScriptingFloatChannel, default=0, const [0,27]}
       hips_tan_ctrl.Scale.X {type=MovieSceneScriptingFloatChannel, default=1, const [1,27]}
       hips_tan_ctrl.Scale.Y {type=MovieSceneScriptingFloatChannel, default=1, const [1,27]}
       hips_tan_ctrl.Scale.Z {type=MovieSceneScriptingFloatChannel, default=1, const [1,27]}
       spine_01_ctrl.Location.X {type=MovieSceneScriptingFloatChannel, default=0, const [0,23]}
       spine_01_ctrl.Location.Y {type=MovieSceneScriptingFloatChannel, default=0, const [0,23]}
       spine_01_ctrl.Location.Z {
        type=MovieSceneScriptingFloatChannel
        default=0
        thinned_from=23
        keys [[0,0.0048],[5,0.003],[25,0.0002],[173,0],[184,-0.0216],[197,-0.2001],[216,-0.0826],[236,-0.0052],[253,0.0118],[269,0.0048],[274,0.003],[294,0.0002]]
       }
       spine_01_ctrl.Rotation.X {type=MovieSceneScriptingFloatChannel, default=0, thinned_from=23, keys [[0,0],[294,0]]}
       spine_01_ctrl.Rotation.Y {type=MovieSceneScriptingFloatChannel, default=0, const [1.5491,23]}
       spine_01_ctrl.Rotation.Z {type=MovieSceneScriptingFloatChannel, default=0, const [-1.7097,23]}
       spine_01_ctrl.Scale.X {type=MovieSceneScriptingFloatChannel, default=1, const [1,23]}
       spine_01_ctrl.Scale.Y {type=MovieSceneScriptingFloatChannel, default=1, const [1,23]}
       spine_01_ctrl.Scale.Z {type=MovieSceneScriptingFloatChannel, default=1, const [1,23]}
       spine_02_ctrl.Location.X {type=MovieSceneScriptingFloatChannel, default=0, const [0,27]}
       spine_02_ctrl.Location.Y {type=MovieSceneScriptingFloatChannel, default=0, const [0,27]}
       spine_02_ctrl.Location.Z {
        type=MovieSceneScriptingFloatChannel
        default=0
        thinned_from=27
        keys [[0,0.0178],[5,0.0109],[25,0.0006],[32,-0.0677],[42,0],[50,0.0164],[76,0.0164],[87,0.0174],[97,0.0244],[111,0.0071],[119,-0.0277],[127,-0.0962],[134,-0.2654],[142,0],[149,0.0056],[159,-0.1281],[166,0],[173,0.085],[184,-0.0496],[197,-0.1043],[216,0],[236,0.0357],[253,0.0436],[269,0.0178],[274,0.0109],[294,0.0006]]
       }
       spine_02_ctrl.Rotation.X {
        type=MovieSceneScriptingFloatChannel
        default=0
        thinned_from=27
        keys [[0,3.0851],[25,3.0851],[32,3.0941],[42,3.157],[50,3.1717],[111,3.1805],[119,3.1755],[127,3.1382],[134,3.0984],[142,3.0851],[294,3.0851]]
       }
       spine_02_ctrl.Rotation.Y {
        type=MovieSceneScriptingFloatChannel
        default=0
        thinned_from=27
        keys [[0,1.7227],[25,1.7227],[32,1.8264],[42,2.5519],[50,2.7167],[68,2.8155],[111,2.8155],[119,2.758],[127,2.3312],[134,1.8748],[142,1.7227],[294,1.7227]]
       }
       spine_02_ctrl.Rotation.Z {
        type=MovieSceneScriptingFloatChannel
        default=0
        thinned_from=27
        keys [[0,0.0626],[25,0.0626],[32,-0.4443],[42,-3.471],[50,-4.7833],[68,-5.5708],[111,-5.5708],[119,-5.2743],[127,-3.074],[134,-0.7216],[142,0.0626],[294,0.0626]]
       }
       spine_02_ctrl.Scale.X {type=MovieSceneScriptingFloatChannel, default=1, const [1,27]}
       spine_02_ctrl.Scale.Y {type=MovieSceneScriptingFloatChannel, default=1, const [1,27]}
       spine_02_ctrl.Scale.Z {type=MovieSceneScriptingFloatChannel, default=1, const [1,27]}
       spine_03_ctrl.Location.X {type=MovieSceneScriptingFloatChannel, default=0, thinned_from=27, keys [[0,0],[149,0],[159,0.0051],[166,0],[294,0]]}
       spine_03_ctrl.Location.Y {type=MovieSceneScriptingFloatChannel, default=0, thinned_from=27, keys [[0,0],[149,0],[159,-0.3315],[166,0],[294,0]]}
       spine_03_ctrl.Location.Z {
        type=MovieSceneScriptingFloatChannel
        default=0
        thinned_from=27
        keys [[0,0.0174],[5,0.0106],[25,0.0006],[32,-0.0677],[42,0],[50,0.0164],[76,0.0164],[87,0.0174],[97,0.0244],[111,0.0071],[119,-0.0277],[127,-0.0962],[134,-0.2654],[142,0],[149,0.0056],[159,-0.1175],[166,0],[173,0.085],[184,-0.0496],[197,-0.1018],[216,0],[236,0.0349],[253,0.0425],[269,0.0174],[274,0.0106],[294,0.0006]]
       }
       spine_03_ctrl.Rotation.X {
        type=MovieSceneScriptingFloatChannel
        default=0
        thinned_from=27
        keys [[0,11.302],[25,11.389],[32,11.5407],[42,12.3055],[50,12.6932],[68,12.9258],[76,12.7276],[87,11.7058],[97,11.3401],[119,11.8324],[127,11.2423],[134,11.1502],[142,11.4531],[149,13.0105],[159,12.3827],[166,11.5405],[173,11.2721],[184,11.3374],[197,11.4469],[216,11.3924],[236,11.2605],[253,11.2212],[274,11.3303],[294,11.389]]
       }
       spine_03_ctrl.Rotation.Y {
        type=MovieSceneScriptingFloatChannel
        default=0
        thinned_from=27
        keys [[0,0.1004],[25,0.1004],[32,0.3419],[42,2.0329],[50,2.4957],[68,2.7734],[76,2.4324],[87,0.5396],[97,0.0459],[111,0.1719],[119,0.2075],[127,0.07],[134,0.0928],[142,0.1004],[149,0.0956],[159,0.0992],[166,0.1004],[294,0.1004]]
       }
       spine_03_ctrl.Rotation.Z {
        type=MovieSceneScriptingFloatChannel
        default=0
        thinned_from=27
        keys [[0,0.1418],[25,0.1418],[32,0.1001],[42,-0.0581],[50,-0.7206],[68,-1.1181],[76,-1.2332],[87,-0.9272],[97,0.0884],[111,1.7166],[119,3.1257],[127,2.2954],[134,0.3894],[142,0.1438],[149,0.6921],[159,0.4281],[166,0.1438],[173,-0.5648],[184,-0.1219],[197,0.0774],[216,0.1418],[294,0.1418]]
       }
       spine_03_ctrl.Scale.X {type=MovieSceneScriptingFloatChannel, default=1, const [1,27]}
       spine_03_ctrl.Scale.Y {type=MovieSceneScriptingFloatChannel, default=1, const [1,27]}
       spine_03_ctrl.Scale.Z {type=MovieSceneScriptingFloatChannel, default=1, const [1,27]}
       spine_04_ctrl.Location.X {type=MovieSceneScriptingFloatChannel, default=0, const [0,27]}
       spine_04_ctrl.Location.Y {type=MovieSceneScriptingFloatChannel, default=0, const [0,27]}
       spine_04_ctrl.Location.Z {type=MovieSceneScriptingFloatChannel, default=0, const [0,27]}
       spine_04_ctrl.Rotation.X {
        type=MovieSceneScriptingFloatChannel
        default=0
        keys [[0,3.2803],[5,3.3433],[25,3.6754],[32,4.186],[42,-3.3482],[50,-8.0186],[68,-10.8208],[76,-11.5398],[87,-13.7336],[97,-14.5368],[111,-14.5368],[119,-13.6525],[127,-7.0898],[134,-0.0733],[142,3.0899],[149,3.8553],[159,2.663],[166,2.2483],[173,2.0598],[184,2.2655],[197,3.3433],[216,3.7025],[236,3.3306],[253,3.1636],[269,3.2803],[274,3.3433],[294,3.6754]]
       }
       spine_04_ctrl.Rotation.Y {
        type=MovieSceneScriptingFloatChannel
        default=0
        keys [[0,4.7719],[5,4.769],[25,5.122],[32,4.8696],[42,3.4277],[50,2.7769],[68,2.3865],[76,2.4216],[87,2.6366],[97,2.6673],[111,2.6673],[119,2.7987],[127,3.7734],[134,4.8156],[142,5.163],[149,7.3144],[159,5.4641],[166,4.2571],[173,3.389],[184,3.734],[197,4.769],[216,5.163],[236,4.9264],[253,4.8057],[269,4.7719],[274,4.769],[294,5.122]]
       }
       spine_04_ctrl.Rotation.Z {
        type=MovieSceneScriptingFloatChannel
        default=0
        thinned_from=27
        keys [[0,4.5782],[25,4.6157],[32,4.534],[42,4.1757],[50,3.9179],[68,3.9108],[76,3.8875],[87,3.7453],[97,3.725],[111,3.725],[119,3.772],[127,4.1205],[134,4.4931],[142,4.6173],[149,6.5235],[159,5.3452],[166,4.6173],[173,4.3103],[184,4.3799],[197,4.5885],[216,4.6173],[236,4.5763],[253,4.5579],[294,4.6157]]
       }
       spine_04_ctrl.Scale.X {type=MovieSceneScriptingFloatChannel, default=1, const [1,27]}
       spine_04_ctrl.Scale.Y {type=MovieSceneScriptingFloatChannel, default=1, const [1,27]}
       spine_04_ctrl.Scale.Z {type=MovieSceneScriptingFloatChannel, default=1, const [1,27]}
       head_ctrl.Location.X {
        type=MovieSceneScriptingFloatChannel
        default=0
        thinned_from=13
        keys [[0,0],[25,0],[32,-0.0015],[42,-0.0071],[50,-0.0092],[68,-0.0095],[76,-0.0076],[87,0],[294,0]]
       }
       head_ctrl.Location.Y {
        type=MovieSceneScriptingFloatChannel
        default=0
        thinned_from=13
        keys [[0,0],[25,0],[32,-0.0179],[42,-0.0841],[50,-0.1093],[68,-0.113],[76,-0.0897],[87,0],[294,0]]
       }
       head_ctrl.Location.Z {
        type=MovieSceneScriptingFloatChannel
        default=0
        thinned_from=13
        keys [[0,0],[25,0],[32,-0.0712],[42,-0.3355],[50,-0.4363],[68,-0.4507],[76,-0.3578],[87,0],[294,0]]
       }
       head_ctrl.Rotation.X {
        type=MovieSceneScriptingFloatChannel
        default=0
        keys [[0,-17.357],[5,-17.385],[25,-16.3405],[32,-16.4819],[42,-15.6071],[50,-15.3155],[68,-14.5785],[76,-13.3679],[87,-7.6978],[97,-6.2842],[104,-5.8073],[111,-5.6186],[119,-7.6241],[127,-15.4964],[134,-21.5203],[142,-20.67],[149,-17.1923],[157,-12.9258],[167,-14.6926],[173,-16.9208],[177,-18.472],[183,-17.8947],[187,-16.2639],[197,-13.4562],[209,-12.3226],[236,-15.2447],[253,-17.0267],[269,-17.357],[274,-17.385],[294,-16.3405]]
       }
       head_ctrl.Rotation.Y {
        type=MovieSceneScriptingFloatChannel
        default=0
        thinned_from=29
        keys [[0,-6.1933],[5,-6.1933],[25,-5.5976],[32,-5.6266],[42,-6.3133],[50,-5.7875],[68,-5.6028],[76,-5.7075],[87,-6.7923],[97,-5.626],[104,-5.2325],[111,-5.0769],[119,-2.3743],[127,-2.2867],[134,-2.2487],[142,-1.3809],[149,0.1101],[159,1.1511],[167,-3.942],[173,-5.4413],[177,-6.2339],[183,-7.1446],[197,-6.1933],[218,-5.5285],[236,-5.9593],[253,-6.1933],[274,-6.1933],[294,-5.5976]]
       }
       head_ctrl.Rotation.Z {
        type=MovieSceneScriptingFloatChannel
        default=0
        thinned_from=29
        keys [[0,-44.5963],[5,-44.5963],[25,-45.0578],[32,-46.071],[42,-49.4462],[50,-50.5713],[68,-51.1325],[76,-43.2843],[87,0.1513],[97,11.2319],[104,9.6503],[111,7.9503],[119,1.8989],[127,-25.8484],[134,-37.8897],[142,-40.4618],[149,-37.7589],[159,-37.4211],[166,-38.7268],[173,-40.5029],[177,-41.8492],[183,-43.5712],[197,-44.5963],[218,-45.1114],[236,-44.7776],[253,-44.5963],[274,-44.5963],[294,-45.0578]]
       }
       head_ctrl.Scale.X {type=MovieSceneScriptingFloatChannel, default=1, const [1,7]}
       head_ctrl.Scale.Y {type=MovieSceneScriptingFloatChannel, default=1, const [1,7]}
       head_ctrl.Scale.Z {type=MovieSceneScriptingFloatChannel, default=1, const [1,7]}
       clavicle_l_ctrl.Location.X {type=MovieSceneScriptingFloatChannel, default=0, const [0,5]}
       clavicle_l_ctrl.Location.Y {type=MovieSceneScriptingFloatChannel, default=0, const [0,5]}
       clavicle_l_ctrl.Location.Z {type=MovieSceneScriptingFloatChannel, default=0, const [0,5]}
       clavicle_l_ctrl.Rotation.X {
        type=MovieSceneScriptingFloatChannel
        default=0
        thinned_from=30
        keys [[0,-2.4424],[5,-2.4424],[25,-2.3426],[32,-2.7764],[134,-2.7764],[142,-4.5857],[149,-3.0039],[159,-2.7764],[185,-2.7764],[197,-2.4424],[216,-2.331],[236,-2.4079],[253,-2.4424],[274,-2.4424],[294,-2.3426]]
       }
       clavicle_l_ctrl.Rotation.Y {
        type=MovieSceneScriptingFloatChannel
        default=0
        thinned_from=27
        keys [[0,-4.3364],[25,-5.1057],[32,-5.9058],[42,-6.9061],[50,-7.4752],[58,-7.6649],[76,-7.6649],[87,-6.2755],[97,-7.0195],[111,-7.3388],[119,-7.1509],[127,-6.2583],[134,-4.5775],[142,-0.9674],[149,3.1532],[159,0.3793],[166,-5.2633],[170,-5.7513],[177,-3.7864],[185,-1.934],[197,-3.7069],[211,-5.1373],[231,-4.2511],[248,-3.0949],[269,-4.3364],[294,-5.1057]]
       }
       clavicle_l_ctrl.Rotation.Z {
        type=MovieSceneScriptingFloatChannel
        default=0
        thinned_from=29
        keys [[0,7.5897],[5,7.7135],[25,7.9207],[32,7.268],[42,4.849],[58,4.4764],[76,4.4764],[87,4.269],[97,3.9857],[111,3.7163],[119,3.2352],[127,1.7856],[134,8.949],[142,15.6205],[149,13.072],[159,11.4363],[170,7.2687],[174,5.7359],[177,5.2184],[185,6.1415],[197,7.4851],[216,7.933],[231,7.6562],[236,7.5045],[253,7.1632],[274,7.7135],[294,7.9207]]
       }
       clavicle_l_ctrl.Scale.X {type=MovieSceneScriptingFloatChannel, default=1, const [1,5]}
       clavicle_l_ctrl.Scale.Y {type=MovieSceneScriptingFloatChannel, default=1, const [1,5]}
       clavicle_l_ctrl.Scale.Z {type=MovieSceneScriptingFloatChannel, default=1, const [1,5]}
       clavicle_r_ctrl.Location.X {type=MovieSceneScriptingFloatChannel, default=0, const [0,6]}
       clavicle_r_ctrl.Location.Y {type=MovieSceneScriptingFloatChannel, default=0, const [0,6]}
       clavicle_r_ctrl.Location.Z {type=MovieSceneScriptingFloatChannel, default=0, const [0,6]}
       clavicle_r_ctrl.Rotation.X {
        type=MovieSceneScriptingFloatChannel
        default=0
        thinned_from=31
        keys [[0,-3.4133],[68,-3.4133],[76,-3.4258],[87,-3.1705],[97,-2.9814],[111,-2.8401],[119,-2.8502],[127,-2.9201],[134,-3.1663],[142,-3.4133],[294,-3.4133]]
       }
       clavicle_r_ctrl.Rotation.Y {
        type=MovieSceneScriptingFloatChannel
        default=0
        thinned_from=28
        keys [[0,-11.731],[25,-12.0164],[32,-12.2144],[76,-12.2144],[87,-11.0608],[97,-10.2066],[111,-9.6641],[119,-9.699],[127,-9.9401],[134,-10.7893],[142,-11.6414],[149,-11.6414],[159,-11.0078],[166,-8.8251],[169,-8.1821],[173,-8.0143],[176,-8.4354],[185,-10.3598],[197,-11.731],[211,-12.0281],[231,-11.7439],[248,-11.249],[269,-11.731],[294,-12.0164]]
       }
       clavicle_r_ctrl.Rotation.Z {
        type=MovieSceneScriptingFloatChannel
        default=0
        thinned_from=31
        keys [[0,2.9292],[5,2.9543],[25,2.9964],[32,2.6201],[42,2.7148],[50,2.9989],[58,3.0555],[68,3.0744],[76,3.0744],[87,0.4387],[97,-1.513],[111,-2.995],[119,-3.1864],[127,-2.4557],[134,0.2721],[142,2.9989],[185,2.9989],[197,3.3341],[216,3.4649],[231,3.2377],[236,3.1154],[253,2.8426],[274,2.9543],[294,2.9964]]
       }
       clavicle_r_ctrl.Scale.X {type=MovieSceneScriptingFloatChannel, default=1, const [1,6]}
       clavicle_r_ctrl.Scale.Y {type=MovieSceneScriptingFloatChannel, default=1, const [1,6]}
       clavicle_r_ctrl.Scale.Z {type=MovieSceneScriptingFloatChannel, default=1, const [1,6]}
       neck_01_ctrl.Location.X {type=MovieSceneScriptingFloatChannel, default=0, const [0,7]}
       neck_01_ctrl.Location.Y {type=MovieSceneScriptingFloatChannel, default=0, const [0,7]}
       neck_01_ctrl.Location.Z {type=MovieSceneScriptingFloatChannel, default=0, const [0,7]}
       neck_01_ctrl.Rotation.X {
        type=MovieSceneScriptingFloatChannel
        default=0
        thinned_from=29
        keys [[0,11.5707],[68,11.5968],[76,11.2789],[87,9.3895],[97,9.4591],[111,9.4918],[119,10.1928],[127,12.2178],[134,13.0966],[142,13.522],[149,13.6073],[159,13.397],[167,11.8218],[173,11.6034],[177,11.724],[183,11.9642],[197,11.5968],[214,11.9547],[236,11.6896],[253,11.5968],[274,11.5685],[294,11.5939]]
       }
       neck_01_ctrl.Rotation.Y {
        type=MovieSceneScriptingFloatChannel
        default=0
        thinned_from=30
        keys [[0,-0.2397],[25,-0.2397],[32,-0.5955],[42,-1.7148],[50,-2.2552],[68,-2.3318],[76,-2.2187],[87,-1.8075],[97,-1.1138],[104,-0.8799],[111,-0.7873],[119,-0.8598],[127,-1.1499],[134,-1.2758],[142,-1.4043],[149,-1.3482],[159,-0.9834],[167,-0.3594],[173,-0.2746],[177,-0.2638],[183,-0.2524],[187,-0.2473],[197,-0.2397],[294,-0.2397]]
       }
       neck_01_ctrl.Rotation.Z {
        type=MovieSceneScriptingFloatChannel
        default=0
        thinned_from=30
        keys [[0,0.1055],[25,0.1055],[32,-1.7945],[42,-7.7709],[50,-10.3352],[68,-11.3716],[76,-8.2237],[87,-4.5663],[97,-4.3005],[104,-4.2109],[111,-4.1754],[119,-4.3768],[127,-3.0216],[134,-2.4335],[142,-0.4027],[149,-2.4071],[159,-1.2789],[167,-0.0846],[173,0.08],[177,0.0902],[183,0.0974],[187,0.1006],[197,0.1055],[253,0.1053],[294,0.1055]]
       }
       neck_01_ctrl.Scale.X {type=MovieSceneScriptingFloatChannel, default=1, const [1,7]}
       neck_01_ctrl.Scale.Y {type=MovieSceneScriptingFloatChannel, default=1, const [1,7]}
       neck_01_ctrl.Scale.Z {type=MovieSceneScriptingFloatChannel, default=1, const [1,7]}
       neck_02_ctrl.Location.X {type=MovieSceneScriptingFloatChannel, default=0, const [0,6]}
       neck_02_ctrl.Location.Y {type=MovieSceneScriptingFloatChannel, default=0, const [0,6]}
       neck_02_ctrl.Location.Z {type=MovieSceneScriptingFloatChannel, default=0, const [0,6]}
       neck_02_ctrl.Rotation.X {type=MovieSceneScriptingFloatChannel, default=0, const [0,6]}
       neck_02_ctrl.Rotation.Y {type=MovieSceneScriptingFloatChannel, default=0, const [0,6]}
       neck_02_ctrl.Rotation.Z {type=MovieSceneScriptingFloatChannel, default=0, const [0,6]}
       neck_02_ctrl.Scale.X {type=MovieSceneScriptingFloatChannel, default=1, const [1,6]}
       neck_02_ctrl.Scale.Y {type=MovieSceneScriptingFloatChannel, default=1, const [1,6]}
       neck_02_ctrl.Scale.Z {type=MovieSceneScriptingFloatChannel, default=1, const [1,6]}
       head_ik_ctrl.Location.X {
        type=MovieSceneScriptingFloatChannel
        default=0
        thinned_from=27
        keys [[0,0.0288],[25,0.0288],[32,0.2259],[42,0.8403],[50,1.0978],[68,1.1827],[76,0.9386],[87,0.6154],[97,0.428],[111,0.3717],[119,0.374],[127,0.5357],[134,0.3812],[142,0.2411],[149,0.4246],[159,0.1535],[167,0.0622],[174,0.035],[185,0.0303],[197,0.0288],[294,0.0288]]
       }
       head_ik_ctrl.Location.Y {
        type=MovieSceneScriptingFloatChannel
        default=0
        thinned_from=27
        keys [[0,1.869],[5,1.8732],[25,1.8732],[32,1.7807],[42,1.6671],[50,1.5674],[68,1.5385],[76,1.6231],[87,1.4926],[97,1.5245],[111,1.5316],[119,1.5314],[127,1.5619],[134,1.628],[142,1.6546],[149,2.1676],[159,2.1491],[167,1.9078],[174,1.8732],[218,1.8732],[236,1.8875],[253,1.8732],[269,1.8691],[274,1.8732],[294,1.8732]]
       }
       head_ik_ctrl.Location.Z {
        type=MovieSceneScriptingFloatChannel
        default=0
        thinned_from=27
        keys [[0,-0.7236],[5,-0.7257],[25,-0.7257],[32,-0.9075],[42,-1.0038],[50,-1.1001],[68,-1.1139],[76,-1.0659],[87,-0.563],[97,-0.5668],[119,-0.5676],[127,-0.5899],[134,-0.6132],[142,-0.621],[149,-0.8922],[159,-0.873],[166,-0.7436],[173,-0.7258],[218,-0.7257],[236,-0.733],[253,-0.7257],[269,-0.7237],[274,-0.7257],[294,-0.7257]]
       }
       head_ik_ctrl.Rotation.X {
        type=MovieSceneScriptingFloatChannel
        default=0
        keys [[0,-8.9818],[5,-8.3934],[25,-7.9977],[32,-7.5209],[42,-7.0072],[50,-5.9266],[68,-5.2029],[76,-4.3009],[87,-0.3174],[97,2.228],[111,4.4484],[119,3.8539],[127,-3.8065],[134,-10.1551],[142,-12.0197],[149,-9.3244],[159,-4.7235],[166,-5.0825],[173,-3.8742],[184,-6.9177],[197,-8.3934],[218,-7.588],[236,-7.1574],[253,-8.9809],[269,-8.9715],[274,-8.3934],[294,-7.9977]]
       }
       head_ik_ctrl.Rotation.Y {
        type=MovieSceneScriptingFloatChannel
        default=0
        keys [[0,1.7961],[5,1.9258],[25,2.4713],[32,0.6202],[42,0.6049],[50,1.7646],[68,1.9869],[76,0.4655],[87,-8.5327],[97,-10.7284],[111,0.007],[119,-9.4875],[127,-2.7609],[134,2.9477],[142,4.2021],[149,6.2987],[159,9.0206],[166,0.7969],[173,1.6737],[184,0.7186],[197,1.9258],[218,2.5474],[236,2.1385],[253,1.815],[269,1.7967],[274,1.9258],[294,2.4713]]
       }
       head_ik_ctrl.Rotation.Z {
        type=MovieSceneScriptingFloatChannel
        default=0
        thinned_from=27
        keys [[0,-44.7895],[5,-44.7725],[25,-45.1566],[32,-48.5766],[42,-57.1895],[50,-60.998],[68,-62.5549],[76,-58.805],[87,-5.5344],[97,3.0172],[111,4.5476],[119,5.0167],[127,-16.372],[134,-40.1568],[142,-40.5305],[149,-42.2045],[159,-39.4189],[166,-43.0324],[173,-40.7052],[184,-43.9266],[197,-44.7725],[218,-45.1993],[236,-44.9304],[253,-44.789],[274,-44.7725],[294,-45.1566]]
       }
       head_ik_ctrl.Scale.X {type=MovieSceneScriptingFloatChannel, default=1, const [1,27]}
       head_ik_ctrl.Scale.Y {type=MovieSceneScriptingFloatChannel, default=1, const [1,27]}
       head_ik_ctrl.Scale.Z {type=MovieSceneScriptingFloatChannel, default=1, const [1,27]}
       chest_ctrl.Location.X {
        type=MovieSceneScriptingFloatChannel
        default=0
        thinned_from=27
        keys [[0,-0.6521],[5,-0.6507],[32,-0.6507],[42,-0.7967],[50,-0.8443],[111,-0.844],[119,-0.8341],[127,-0.7594],[134,-0.678],[142,-0.6507],[216,-0.6507],[236,-0.6535],[253,-0.6542],[274,-0.6507],[294,-0.6507]]
       }
       chest_ctrl.Location.Y {
        type=MovieSceneScriptingFloatChannel
        default=0
        thinned_from=27
        keys [[0,0.4545],[5,0.4535],[32,0.4535],[42,0.4944],[50,0.5122],[111,0.5137],[119,0.5096],[134,0.4598],[142,0.4535],[216,0.4535],[236,0.4554],[253,0.4559],[274,0.4535],[294,0.4535]]
       }
       chest_ctrl.Location.Z {
        type=MovieSceneScriptingFloatChannel
        default=0
        thinned_from=27
        keys [[0,0.0606],[5,0.0206],[25,0.022],[32,0.0206],[42,0.0128],[50,0.0099],[111,0.01],[119,0.0106],[127,0.015],[134,0.0193],[142,0.0206],[216,0.0206],[236,0.0859],[253,0.1185],[269,0.0606],[274,0.0206],[294,0.022]]
       }
       chest_ctrl.Rotation.X {
        type=MovieSceneScriptingFloatChannel
        default=0
        thinned_from=27
        keys [[0,14.3972],[25,14.4842],[32,14.2148],[42,15.4056],[50,15.8067],[68,16.0278],[76,15.8382],[87,14.5135],[97,14.5135],[111,14.7381],[119,14.4962],[134,14.2525],[142,14.2148],[149,16.106],[159,14.2148],[166,14.2148],[173,14.3268],[184,14.1996],[197,14.4219],[216,14.2148],[236,14.3557],[253,14.3164],[274,14.4219],[294,14.4842]]
       }
       chest_ctrl.Rotation.Y {
        type=MovieSceneScriptingFloatChannel
        default=0
        thinned_from=27
        keys [[0,3.3643],[5,3.3678],[32,3.3642],[42,6.0961],[50,7.005],[68,7.0862],[76,6.738],[87,4.2999],[97,4.2999],[111,4.4922],[119,4.251],[127,3.8872],[134,3.4954],[142,3.3642],[149,3.3593],[166,3.3642],[173,3.4022],[184,3.3785],[197,3.3678],[269,3.3643],[274,3.3678],[294,3.3643]]
       }
       chest_ctrl.Rotation.Z {
        type=MovieSceneScriptingFloatChannel
        default=0
        thinned_from=27
        keys [[0,-1.4998],[5,-1.5642],[25,-1.4998],[32,-1.4978],[42,-5.0166],[50,-6.1897],[68,-6.3813],[76,-6.2553],[87,-5.3766],[97,-5.3766],[111,-6.5911],[119,-5.1724],[127,-3.6573],[134,-2.0376],[142,-1.4978],[149,-1.4948],[166,-1.4978],[173,-2.2066],[184,-1.7636],[197,-1.5642],[216,-1.4978],[269,-1.4998],[274,-1.5642],[294,-1.4998]]
       }
       chest_ctrl.Scale.X {type=MovieSceneScriptingFloatChannel, default=1, const [1,27]}
       chest_ctrl.Scale.Y {type=MovieSceneScriptingFloatChannel, default=1, const [1,27]}
       chest_ctrl.Scale.Z {type=MovieSceneScriptingFloatChannel, default=1, const [1,27]}
       chest_tan_ctrl.Location.X {
        type=MovieSceneScriptingFloatChannel
        default=0
        thinned_from=27
        keys [[0,0.5896],[5,0.5874],[32,0.5891],[42,0.3227],[50,0.2333],[68,0.2157],[76,0.2787],[87,0.7194],[97,0.7194],[111,0.6657],[119,0.7126],[127,0.662],[134,0.6074],[142,0.5891],[149,0.59],[167,0.5891],[174,0.5708],[185,0.5822],[197,0.5874],[236,0.5904],[269,0.5896],[274,0.5874],[294,0.5891]]
       }
       chest_tan_ctrl.Location.Y {
        type=MovieSceneScriptingFloatChannel
        default=-0.0303
        thinned_from=27
        keys [[0,2.2408],[25,2.2539],[32,2.2094],[42,2.4308],[50,2.5037],[68,2.542],[76,2.5075],[87,2.263],[97,2.263],[111,2.3301],[119,2.2604],[127,2.2401],[134,2.2173],[142,2.2094],[149,2.5197],[159,2.2094],[167,2.2094],[174,2.2418],[185,2.2122],[197,2.2449],[216,2.2094],[236,2.2339],[253,2.2293],[274,2.2449],[294,2.2539]]
       }
       chest_tan_ctrl.Location.Z {
        type=MovieSceneScriptingFloatChannel
        default=0.265
        thinned_from=27
        keys [[0,0.7023],[5,0.7158],[25,0.7187],[32,0.7032],[42,0.7562],[50,0.7762],[68,0.7909],[76,0.7791],[87,0.7106],[97,0.7106],[111,0.7312],[119,0.7103],[134,0.7044],[142,0.7032],[149,0.8193],[159,0.7032],[167,0.7032],[174,0.7137],[185,0.7038],[197,0.7158],[216,0.7032],[236,0.693],[253,0.6805],[269,0.7023],[274,0.7158],[294,0.7187]]
       }
       chest_tan_ctrl.Rotation.X {type=MovieSceneScriptingFloatChannel, default=0, thinned_from=27, keys [[0,0],[294,0]]}
       chest_tan_ctrl.Rotation.Y {type=MovieSceneScriptingFloatChannel, default=0, thinned_from=27, keys [[0,0],[167,0],[174,0.012],[185,0],[294,0]]}
       chest_tan_ctrl.Rotation.Z {type=MovieSceneScriptingFloatChannel, default=0, thinned_from=27, keys [[0,0],[167,0],[174,0.5735],[185,0.139],[197,0],[294,0]]}
       chest_tan_ctrl.Scale.X {type=MovieSceneScriptingFloatChannel, default=1, const [1,27]}
       chest_tan_ctrl.Scale.Y {type=MovieSceneScriptingFloatChannel, default=1, const [1,27]}
       chest_tan_ctrl.Scale.Z {type=MovieSceneScriptingFloatChannel, default=1, const [1,27]}
       Weight {type=MovieSceneScriptingFloatChannel, default=1}
       leg_l_pv_ik_ctrl.Space {
        type=MovieSceneScriptingSpaceChannel
        const [
         "<Struct 'MovieSceneControlRigSpaceBaseKey' (0x49dab460f00) {space_type: ControlRig, control_rig_element: {type: Control, name: \"hips_ctrl\"}}>"
         1
        ]
       }
       leg_r_pv_ik_ctrl.Space {
        type=MovieSceneScriptingSpaceChannel
        const [
         "<Struct 'MovieSceneControlRigSpaceBaseKey' (0x49dab4627a0) {space_type: ControlRig, control_rig_element: {type: Control, name: \"hips_ctrl\"}}>"
         1
        ]
       }
       hand_r_fk_ctrl.Space {type=MovieSceneScriptingSpaceChannel}
      ]
     }
     {
      type=MovieSceneControlRigParameterSection
      channels [
       foot_l_ik_ctrl.Location.X {type=MovieSceneScriptingFloatChannel, default=0, keys [[-1,0],[0,-3.4743],[268,-3.4743]]}
       foot_l_ik_ctrl.Location.Y {type=MovieSceneScriptingFloatChannel, default=0, keys [[-1,0],[0,3.7754],[268,3.7754]]}
       foot_l_ik_ctrl.Location.Z {type=MovieSceneScriptingFloatChannel, default=0, const [0,3]}
       foot_l_ik_ctrl.Rotation.X {type=MovieSceneScriptingFloatChannel, default=0, const [0,3]}
       foot_l_ik_ctrl.Rotation.Y {type=MovieSceneScriptingFloatChannel, default=0, const [0,3]}
       foot_l_ik_ctrl.Rotation.Z {type=MovieSceneScriptingFloatChannel, default=0, const [0,3]}
       foot_l_ik_ctrl.Scale.X {type=MovieSceneScriptingFloatChannel, default=0, const [0,3]}
       foot_l_ik_ctrl.Scale.Y {type=MovieSceneScriptingFloatChannel, default=0, const [0,3]}
       foot_l_ik_ctrl.Scale.Z {type=MovieSceneScriptingFloatChannel, default=0, const [0,3]}
       leg_l_softness {type=MovieSceneScriptingFloatChannel, default=-3.4268, keys [[-1,-3.383],[0,-3.4112],[268,-3.4112]]}
       foot_r_ik_ctrl.Location.X {type=MovieSceneScriptingFloatChannel, default=0, keys [[-1,0],[0,-2.5646],[268,-2.5646]]}
       foot_r_ik_ctrl.Location.Y {type=MovieSceneScriptingFloatChannel, default=0, keys [[-1,0],[0,4.7079],[268,4.7079]]}
       foot_r_ik_ctrl.Location.Z {type=MovieSceneScriptingFloatChannel, default=0, const [0,3]}
       foot_r_ik_ctrl.Rotation.X {type=MovieSceneScriptingFloatChannel, default=0, const [0,3]}
       foot_r_ik_ctrl.Rotation.Y {type=MovieSceneScriptingFloatChannel, default=0, const [0,3]}
       foot_r_ik_ctrl.Rotation.Z {type=MovieSceneScriptingFloatChannel, default=0, const [0,3]}
       foot_r_ik_ctrl.Scale.X {type=MovieSceneScriptingFloatChannel, default=0, const [0,3]}
       foot_r_ik_ctrl.Scale.Y {type=MovieSceneScriptingFloatChannel, default=0, const [0,3]}
       foot_r_ik_ctrl.Scale.Z {type=MovieSceneScriptingFloatChannel, default=0, const [0,3]}
       leg_r_softness {type=MovieSceneScriptingFloatChannel, default=0, const [0,3]}
       body_ctrl.Location.X {type=MovieSceneScriptingFloatChannel, default=0, thinned_from=4, keys [[-1,0],[0,1.1668],[268,1.1668]]}
       body_ctrl.Location.Y {type=MovieSceneScriptingFloatChannel, default=0, thinned_from=4, keys [[-1,0],[0,3.5],[268,3.5]]}
       body_ctrl.Location.Z {type=MovieSceneScriptingFloatChannel, default=0, thinned_from=4, keys [[-1,0],[0,1.5339],[268,1.5339]]}
       body_ctrl.Rotation.X {type=MovieSceneScriptingFloatChannel, default=0, thinned_from=4, keys [[-1,0],[0,1.6416],[268,1.6416]]}
       body_ctrl.Rotation.Y {type=MovieSceneScriptingFloatChannel, default=0, const [0,4]}
       body_ctrl.Rotation.Z {type=MovieSceneScriptingFloatChannel, default=0, const [0,4]}
       body_ctrl.Scale.X {type=MovieSceneScriptingFloatChannel, default=0, const [0,4]}
       body_ctrl.Scale.Y {type=MovieSceneScriptingFloatChannel, default=0, const [0,4]}
       body_ctrl.Scale.Z {type=MovieSceneScriptingFloatChannel, default=0, const [0,4]}
       hips_ctrl.Location.X {type=MovieSceneScriptingFloatChannel, default=0, const [0,3]}
       hips_ctrl.Location.Y {type=MovieSceneScriptingFloatChannel, default=0, const [0,3]}
       hips_ctrl.Location.Z {type=MovieSceneScriptingFloatChannel, default=0, const [0,3]}
       hips_ctrl.Rotation.X {type=MovieSceneScriptingFloatChannel, default=0.0039, keys [[-1,0.0039],[0,6.7475],[268,6.7475]]}
       hips_ctrl.Rotation.Y {type=MovieSceneScriptingFloatChannel, default=0.0007, keys [[-1,0.0007],[0,2.3502],[268,2.3502]]}
       hips_ctrl.Rotation.Z {type=MovieSceneScriptingFloatChannel, default=0, const [0,3]}
       hips_ctrl.Scale.X {type=MovieSceneScriptingFloatChannel, default=0, const [0,3]}
       hips_ctrl.Scale.Y {type=MovieSceneScriptingFloatChannel, default=0, const [0,3]}
       hips_ctrl.Scale.Z {type=MovieSceneScriptingFloatChannel, default=0, const [0,3]}
       spine_03_ctrl.Location.X {type=MovieSceneScriptingFloatChannel, default=0, const [0,3]}
       spine_03_ctrl.Location.Y {type=MovieSceneScriptingFloatChannel, default=0, const [0,3]}
       spine_03_ctrl.Location.Z {type=MovieSceneScriptingFloatChannel, default=0, const [0,3]}
       spine_03_ctrl.Rotation.X {type=MovieSceneScriptingFloatChannel, default=0, keys [[-1,0],[0,-7.7142],[268,-7.7142]]}
       spine_03_ctrl.Rotation.Y {type=MovieSceneScriptingFloatChannel, default=0, const [0,3]}
       spine_03_ctrl.Rotation.Z {type=MovieSceneScriptingFloatChannel, default=0, const [0,3]}
       spine_03_ctrl.Scale.X {type=MovieSceneScriptingFloatChannel, default=0, const [0,3]}
       spine_03_ctrl.Scale.Y {type=MovieSceneScriptingFloatChannel, default=0, const [0,3]}
       spine_03_ctrl.Scale.Z {type=MovieSceneScriptingFloatChannel, default=0, const [0,3]}
       spine_04_ctrl.Location.X {type=MovieSceneScriptingFloatChannel, default=0, const [0,3]}
       spine_04_ctrl.Location.Y {type=MovieSceneScriptingFloatChannel, default=0, const [0,3]}
       spine_04_ctrl.Location.Z {type=MovieSceneScriptingFloatChannel, default=0, const [0,3]}
       spine_04_ctrl.Rotation.X {type=MovieSceneScriptingFloatChannel, default=0, keys [[-1,0],[0,-3.0274],[268,-3.0274]]}
       spine_04_ctrl.Rotation.Y {type=MovieSceneScriptingFloatChannel, default=0, const [0,3]}
       spine_04_ctrl.Rotation.Z {type=MovieSceneScriptingFloatChannel, default=0, const [0,3]}
       spine_04_ctrl.Scale.X {type=MovieSceneScriptingFloatChannel, default=0, const [0,3]}
       spine_04_ctrl.Scale.Y {type=MovieSceneScriptingFloatChannel, default=0, const [0,3]}
       spine_04_ctrl.Scale.Z {type=MovieSceneScriptingFloatChannel, default=0, const [0,3]}
       head_ctrl.Location.X {type=MovieSceneScriptingFloatChannel, default=0, const [0,3]}
       head_ctrl.Location.Y {type=MovieSceneScriptingFloatChannel, default=0, const [0,3]}
       head_ctrl.Location.Z {type=MovieSceneScriptingFloatChannel, default=0, const [0,3]}
       head_ctrl.Rotation.X {type=MovieSceneScriptingFloatChannel, default=0, keys [[-1,0],[0,8.893],[268,8.893]]}
       head_ctrl.Rotation.Y {type=MovieSceneScriptingFloatChannel, default=0, keys [[-1,0],[0,2.5412],[268,2.5412]]}
       head_ctrl.Rotation.Z {type=MovieSceneScriptingFloatChannel, default=0, keys [[-1,0],[0,2.433],[268,2.433]]}
       head_ctrl.Scale.X {type=MovieSceneScriptingFloatChannel, default=0, const [0,3]}
       head_ctrl.Scale.Y {type=MovieSceneScriptingFloatChannel, default=0, const [0,3]}
       head_ctrl.Scale.Z {type=MovieSceneScriptingFloatChannel, default=0, const [0,3]}
       neck_01_ctrl.Location.X {type=MovieSceneScriptingFloatChannel, default=0, const [0,3]}
       neck_01_ctrl.Location.Y {type=MovieSceneScriptingFloatChannel, default=0, const [0,3]}
       neck_01_ctrl.Location.Z {type=MovieSceneScriptingFloatChannel, default=0, const [0,3]}
       neck_01_ctrl.Rotation.X {type=MovieSceneScriptingFloatChannel, default=0, keys [[-1,0],[0,0.0183],[268,0.0183]]}
       neck_01_ctrl.Rotation.Y {type=MovieSceneScriptingFloatChannel, default=0, keys [[-1,0],[0,3.4487],[268,3.4487]]}
       neck_01_ctrl.Rotation.Z {type=MovieSceneScriptingFloatChannel, default=0, keys [[-1,0],[0,0.7068],[268,0.7068]]}
       neck_01_ctrl.Scale.X {type=MovieSceneScriptingFloatChannel, default=0, const [0,3]}
       neck_01_ctrl.Scale.Y {type=MovieSceneScriptingFloatChannel, default=0, const [0,3]}
       neck_01_ctrl.Scale.Z {type=MovieSceneScriptingFloatChannel, default=0, const [0,3]}
       Weight {type=MovieSceneScriptingFloatChannel, default=1}
      ]
     }
    ]
   }
   {
    type=MovieSceneControlRigParameterTrack
    display_name=Fortnite_3L_Face_CtrlRig
    sections [{
      type=MovieSceneControlRigParameterSection
      channels [
       CTRL_C_eyesAim.Location.X {type=MovieSceneScriptingFloatChannel, default=0, const [0,6]}
       CTRL_C_eyesAim.Location.Y {type=MovieSceneScriptingFloatChannel, default=0, const [0,6]}
       CTRL_C_eyesAim.Location.Z {type=MovieSceneScriptingFloatChannel, default=0, const [0,6]}
       CTRL_C_eyesAim.Rotation.X {type=MovieSceneScriptingFloatChannel, default=0, const [0,6]}
       CTRL_C_eyesAim.Rotation.Y {type=MovieSceneScriptingFloatChannel, default=0, const [0,6]}
       CTRL_C_eyesAim.Rotation.Z {type=MovieSceneScriptingFloatChannel, default=0, const [0,6]}
       CTRL_C_eyesAim.Scale.X {type=MovieSceneScriptingFloatChannel, default=1, const [1,6]}
       CTRL_C_eyesAim.Scale.Y {type=MovieSceneScriptingFloatChannel, default=1, const [1,6]}
       CTRL_C_eyesAim.Scale.Z {type=MovieSceneScriptingFloatChannel, default=1, const [1,6]}
       CTRL_L_eyeAim.Location.X {type=MovieSceneScriptingFloatChannel, default=0, const [0,6]}
       CTRL_L_eyeAim.Location.Y {type=MovieSceneScriptingFloatChannel, default=0, const [0,6]}
       CTRL_L_eyeAim.Location.Z {type=MovieSceneScriptingFloatChannel, default=0, const [0,6]}
       CTRL_L_eyeAim.Rotation.X {type=MovieSceneScriptingFloatChannel, default=0, const [0,6]}
       CTRL_L_eyeAim.Rotation.Y {type=MovieSceneScriptingFloatChannel, default=0, const [0,6]}
       CTRL_L_eyeAim.Rotation.Z {type=MovieSceneScriptingFloatChannel, default=0, const [0,6]}
       CTRL_L_eyeAim.Scale.X {type=MovieSceneScriptingFloatChannel, default=1, const [1,6]}
       CTRL_L_eyeAim.Scale.Y {type=MovieSceneScriptingFloatChannel, default=1, const [1,6]}
       CTRL_L_eyeAim.Scale.Z {type=MovieSceneScriptingFloatChannel, default=1, const [1,6]}
       CTRL_R_eyeAim.Location.X {type=MovieSceneScriptingFloatChannel, default=0, const [0,6]}
       CTRL_R_eyeAim.Location.Y {type=MovieSceneScriptingFloatChannel, default=0, const [0,6]}
       CTRL_R_eyeAim.Location.Z {type=MovieSceneScriptingFloatChannel, default=0, const [0,6]}
       CTRL_R_eyeAim.Rotation.X {type=MovieSceneScriptingFloatChannel, default=0, const [0,6]}
       CTRL_R_eyeAim.Rotation.Y {type=MovieSceneScriptingFloatChannel, default=0, const [0,6]}
       CTRL_R_eyeAim.Rotation.Z {type=MovieSceneScriptingFloatChannel, default=0, const [0,6]}
       CTRL_R_eyeAim.Scale.X {type=MovieSceneScriptingFloatChannel, default=1, const [1,6]}
       CTRL_R_eyeAim.Scale.Y {type=MovieSceneScriptingFloatChannel, default=1, const [1,6]}
       CTRL_R_eyeAim.Scale.Z {type=MovieSceneScriptingFloatChannel, default=1, const [1,6]}
       CTRL_convergenceSwitch {type=MovieSceneScriptingFloatChannel, default=0, const [0,6]}
       CTRL_Master.Location.X {type=MovieSceneScriptingFloatChannel, default=0, const [0,6]}
       CTRL_Master.Location.Y {type=MovieSceneScriptingFloatChannel, default=0, const [0,6]}
       CTRL_Master.Location.Z {type=MovieSceneScriptingFloatChannel, default=0, const [0,6]}
       CTRL_Master.Rotation.X {type=MovieSceneScriptingFloatChannel, default=0, const [0,6]}
       CTRL_Master.Rotation.Y {type=MovieSceneScriptingFloatChannel, default=0, const [0,6]}
       CTRL_Master.Rotation.Z {type=MovieSceneScriptingFloatChannel, default=0, const [0,6]}
       CTRL_Master.Scale.X {type=MovieSceneScriptingFloatChannel, default=1, const [1,6]}
       CTRL_Master.Scale.Y {type=MovieSceneScriptingFloatChannel, default=1, const [1,6]}
       CTRL_Master.Scale.Z {type=MovieSceneScriptingFloatChannel, default=1, const [1,6]}
       CTRL_C_jaw.X {type=MovieSceneScriptingFloatChannel, default=0, const [0,6]}
       CTRL_C_jaw.Y {type=MovieSceneScriptingFloatChannel, default=0, const [0,6]}
       CTRL_C_jaw_openExtreme {type=MovieSceneScriptingFloatChannel, default=0, const [0,6]}
       CTRL_C_jaw_fwdBack {type=MovieSceneScriptingFloatChannel, default=0, const [0,6]}
       CTRL_R_jaw_ChinRaiseD {type=MovieSceneScriptingFloatChannel, default=0, const [0,6]}
       CTRL_R_jaw_ChinRaiseU {type=MovieSceneScriptingFloatChannel, default=0, const [0,6]}
       CTRL_L_jaw_ChinRaiseU {type=MovieSceneScriptingFloatChannel, default=0, const [0,6]}
       CTRL_L_jaw_ChinRaiseD {type=MovieSceneScriptingFloatChannel, default=0, const [0,6]}
       CTRL_R_mouth_lipsBlow {type=MovieSceneScriptingFloatChannel, default=0, const [0,6]}
       CTRL_L_mouth_lipsTogetherU {type=MovieSceneScriptingFloatChannel, default=0, const [0,6]}
       CTRL_L_mouth_lipsTogetherD {type=MovieSceneScriptingFloatChannel, default=0, const [0,6]}
       CTRL_R_mouth_lipsTogetherU {type=MovieSceneScriptingFloatChannel, default=0, const [0,6]}
       CTRL_L_mouth_lipsBlow {type=MovieSceneScriptingFloatChannel, default=0, const [0,6]}
       CTRL_R_mouth_lipsTogetherD {type=MovieSceneScriptingFloatChannel, default=0, const [0,6]}
       CTRL_C_tongue_roll.X {type=MovieSceneScriptingFloatChannel, default=0, const [0,6]}
       CTRL_C_tongue_roll.Y {type=MovieSceneScriptingFloatChannel, default=0, const [0,6]}
       CTRL_C_tongue.X {type=MovieSceneScriptingFloatChannel, default=0, const [0,6]}
       CTRL_C_tongue.Y {type=MovieSceneScriptingFloatChannel, default=0, const [0,6]}
       CTRL_C_tongue_narrowWide {type=MovieSceneScriptingFloatChannel, default=0, const [0,6]}
       CTRL_C_tongue_inOut {type=MovieSceneScriptingFloatChannel, default=0, const [0,6]}
       CTRL_R_nose.X {type=MovieSceneScriptingFloatChannel, default=0, const [0,6]}
       CTRL_R_nose.Y {type=MovieSceneScriptingFloatChannel, default=0, const [0,6]}
       CTRL_L_nose.X {type=MovieSceneScriptingFloatChannel, default=0, const [0,6]}
       CTRL_L_nose.Y {type=MovieSceneScriptingFloatChannel, default=0, const [0,6]}
       CTRL_L_eye.X {type=MovieSceneScriptingFloatChannel, default=0, const [0,6]}
       CTRL_L_eye.Y {type=MovieSceneScriptingFloatChannel, default=0, const [0,6]}
       CTRL_L_eye_blink {type=MovieSceneScriptingFloatChannel, default=0, keys [[0,0],[23,0],[42,1],[50,0.0082],[58,0],[76,1],[87,0],[117,1],[132,0],[180,1],[198,0]]}
       CTRL_R_eye.X {type=MovieSceneScriptingFloatChannel, default=0, const [0,6]}
       CTRL_R_eye.Y {type=MovieSceneScriptingFloatChannel, default=0, const [0,6]}
       CTRL_R_eye_blink {type=MovieSceneScriptingFloatChannel, default=0, keys [[0,0],[23,0],[42,1],[50,0.0082],[58,0],[76,1],[87,0],[117,1],[132,0],[180,1],[198,0]]}
       CTRL_C_eye.X {
        type=MovieSceneScriptingFloatChannel
        default=0
        thinned_from=12
        keys [[0,-0.172],[23,-0.172],[42,0],[76,0],[87,-0.6184],[109,-0.6184],[132,-0.236],[140,-0.236],[163,0],[198,-0.172]]
       }
       CTRL_C_eye.Y {
        type=MovieSceneScriptingFloatChannel
        default=0
        keys [[0,-0.2462],[23,-0.2462],[42,0],[50,-0.3098],[58,-0.3124],[76,-0.3124],[87,-0.3633],[109,-0.3633],[132,0.0076],[140,0.0076],[163,0],[198,-0.2462]]
       }
       CTRL_C_eyesConvergence {type=MovieSceneScriptingFloatChannel, default=0, const [0,6]}
       CTRL_L_eye_cheekRaise {type=MovieSceneScriptingFloatChannel, default=0, thinned_from=6, keys [[0,0.3607],[58,0.3607],[198,0]]}
       CTRL_L_eye_squintInner {type=MovieSceneScriptingFloatChannel, default=0, const [0,6]}
       CTRL_L_nose_wrinkleUpper {type=MovieSceneScriptingFloatChannel, default=0, const [0,6]}
       CTRL_R_nose_wrinkleUpper {type=MovieSceneScriptingFloatChannel, default=0, const [0,6]}
       CTRL_R_eye_squintInner {type=MovieSceneScriptingFloatChannel, default=0, const [0,6]}
       CTRL_R_eye_cheekRaise {type=MovieSceneScriptingFloatChannel, default=0, thinned_from=6, keys [[0,0.3607],[58,0.3607],[198,0]]}
       CTRL_C_eyelidsReduce {type=MovieSceneScriptingFloatChannel, default=0, const [0,6]}
       CTRL_R_brow_lateral {type=MovieSceneScriptingFloatChannel, default=0, const [0.3677,10]}
       CTRL_L_brow_lateral {type=MovieSceneScriptingFloatChannel, default=0, const [0.3677,9]}
       CTRL_R_brow_down {
        type=MovieSceneScriptingFloatChannel
        default=0
        keys [[0,0],[23,0],[42,0.4938],[50,0.004],[58,0],[76,0.4938],[87,0],[117,0.4938],[132,0],[198,0]]
       }
       CTRL_L_brow_down {
        type=MovieSceneScriptingFloatChannel
        default=0
        keys [[0,0],[23,0],[42,0.4938],[50,0.004],[58,0],[76,0.4938],[87,0],[117,0.4938],[132,0],[198,0]]
       }
       CTRL_L_brow_raiseOut {type=MovieSceneScriptingFloatChannel, default=0, thinned_from=10, keys [[0,0.2453],[76,0.2453],[87,0],[117,0.2453],[198,0.2453]]}
       CTRL_L_brow_raiseIn {type=MovieSceneScriptingFloatChannel, default=0, const [0,10]}
       CTRL_R_brow_raiseIn {type=MovieSceneScriptingFloatChannel, default=0, thinned_from=10, keys [[0,0.4229],[23,0],[198,0]]}
       CTRL_R_brow_raiseOut {type=MovieSceneScriptingFloatChannel, default=0, thinned_from=10, keys [[0,0],[76,0],[87,0.2444],[117,0],[198,0]]}
       CTRL_lookAtSwitch {type=MovieSceneScriptingFloatChannel, default=0, const [0,6]}
       CTRL_C_teethD.X {type=MovieSceneScriptingFloatChannel, default=0, const [0,6]}
       CTRL_C_teethD.Y {type=MovieSceneScriptingFloatChannel, default=0, const [0,6]}
       CTRL_C_teethU.X {type=MovieSceneScriptingFloatChannel, default=0, const [0,6]}
       CTRL_C_teethU.Y {type=MovieSceneScriptingFloatChannel, default=0, const [0,6]}
       CTRL_C_teeth_fwdBackU {type=MovieSceneScriptingFloatChannel, default=0, const [0,6]}
       CTRL_L_mouth_corner.X {type=MovieSceneScriptingFloatChannel, default=0, const [0,6]}
       CTRL_L_mouth_corner.Y {type=MovieSceneScriptingFloatChannel, default=0, const [0,6]}
       CTRL_R_mouth_corner.X {type=MovieSceneScriptingFloatChannel, default=0, const [0,6]}
       CTRL_R_mouth_corner.Y {type=MovieSceneScriptingFloatChannel, default=0, const [0,6]}
       CTRL_C_teeth_fwdBackD {type=MovieSceneScriptingFloatChannel, default=0, const [0,6]}
       CTRL_R_mouth_cornerSharpnessD {type=MovieSceneScriptingFloatChannel, default=0, const [0,6]}
       CTRL_R_mouth_cornerSharpnessU {type=MovieSceneScriptingFloatChannel, default=0, const [0,6]}
       CTRL_L_eye_eyelidU.X {type=MovieSceneScriptingFloatChannel, default=0, const [0,6]}
       CTRL_L_eye_eyelidU.Y {type=MovieSceneScriptingFloatChannel, default=0, const [0,6]}
       CTRL_L_mouth_cornerSharpnessD {type=MovieSceneScriptingFloatChannel, default=0, const [0,6]}
       CTRL_L_eye_eyelidD.X {type=MovieSceneScriptingFloatChannel, default=0, const [0,6]}
       CTRL_L_eye_eyelidD.Y {type=MovieSceneScriptingFloatChannel, default=0, const [0,6]}
       CTRL_L_mouth_cornerSharpnessU {type=MovieSceneScriptingFloatChannel, default=0, const [0,6]}
       CTRL_R_mouth_lipsTowardsTeethD {type=MovieSceneScriptingFloatChannel, default=0, const [0,6]}
       CTRL_L_mouth_lipsTowardsTeethD {type=MovieSceneScriptingFloatChannel, default=0, const [0,6]}
       CTRL_R_eye_eyelidU.X {type=MovieSceneScriptingFloatChannel, default=0, const [0,6]}
       CTRL_R_eye_eyelidU.Y {type=MovieSceneScriptingFloatChannel, default=0, const [0,6]}
       CTRL_L_mouth_lipsTowardsTeethU {type=MovieSceneScriptingFloatChannel, default=0, const [0,6]}
       CTRL_R_eye_eyelidD.X {type=MovieSceneScriptingFloatChannel, default=0, const [0,6]}
       CTRL_R_eye_eyelidD.Y {type=MovieSceneScriptingFloatChannel, default=0, const [0,6]}
       CTRL_R_mouth_lipsTowardsTeethU {type=MovieSceneScriptingFloatChannel, default=0, const [0,6]}
       CTRL_faceGUIfollowHead {type=MovieSceneScriptingFloatChannel, default=0, const [0,6]}
       CTRL_eyesAimFollowHead {type=MovieSceneScriptingFloatChannel, default=0, const [0,6]}
       CTRL_C_nose.X {type=MovieSceneScriptingFloatChannel, default=0, const [0,6]}
       CTRL_C_nose.Y {type=MovieSceneScriptingFloatChannel, default=0, const [0,6]}
       CTRL_C_teeth_rotateU {type=MovieSceneScriptingFloatChannel, default=0, const [0,6]}
       CTRL_C_teeth_rotateD {type=MovieSceneScriptingFloatChannel, default=0, const [0,6]}
       CTRL_L_brow_inner {type=MovieSceneScriptingFloatChannel, default=0, thinned_from=10, keys [[0,-1],[42,-1],[50,-0.7542],[58,-0.7522],[76,-1],[198,-1]]}
       CTRL_R_brow_inner {type=MovieSceneScriptingFloatChannel, default=0, thinned_from=10, keys [[0,-1],[76,-1],[87,-0.752],[117,-1],[198,-1]]}
       CTRL_L_brow_outer {
        type=MovieSceneScriptingFloatChannel
        default=0
        thinned_from=10
        keys [[0,-0.4314],[42,-0.4314],[50,-0.116],[58,-0.1134],[76,-0.4314],[198,-0.4314]]
       }
       CTRL_R_brow_outer {type=MovieSceneScriptingFloatChannel, default=0, thinned_from=10, keys [[0,-0.4314],[76,-0.4314],[87,-0.11],[117,-0.4314],[198,-0.4314]]}
       CTRL_C_mouth.X {type=MovieSceneScriptingFloatChannel, default=0, const [0,6]}
       CTRL_C_mouth.Y {type=MovieSceneScriptingFloatChannel, default=0, const [0,6]}
       CTRL_L_mouth_upperLipRaise {type=MovieSceneScriptingFloatChannel, default=0, const [0,6]}
       CTRL_L_mouth_cornerPull {type=MovieSceneScriptingFloatChannel, default=0, const [0,6]}
       CTRL_L_mouth_dimple {type=MovieSceneScriptingFloatChannel, default=0, const [0,6]}
       CTRL_L_mouth_cornerDepress {type=MovieSceneScriptingFloatChannel, default=0, const [0,6]}
       CTRL_L_mouth_stretch {type=MovieSceneScriptingFloatChannel, default=0, const [0,6]}
       CTRL_L_mouth_lowerLipDepress {type=MovieSceneScriptingFloatChannel, default=0, const [0,6]}
       CTRL_R_mouth_lowerLipDepress {type=MovieSceneScriptingFloatChannel, default=0, const [0,6]}
       CTRL_R_mouth_stretch {type=MovieSceneScriptingFloatChannel, default=0, const [0,6]}
       CTRL_L_mouth_towardsD {type=MovieSceneScriptingFloatChannel, default=0, const [0,6]}
       CTRL_R_mouth_funnelU {type=MovieSceneScriptingFloatChannel, default=0, const [0,6]}
       CTRL_L_mouth_towardsU {type=MovieSceneScriptingFloatChannel, default=0, const [0,6]}
       CTRL_R_mouth_lipBiteD {type=MovieSceneScriptingFloatChannel, default=0, const [0,6]}
       CTRL_R_mouth_funnelD {type=MovieSceneScriptingFloatChannel, default=0, const [0,6]}
       CTRL_L_mouth_pressU {type=MovieSceneScriptingFloatChannel, default=0, const [0,6]}
       CTRL_L_mouth_funnelD {type=MovieSceneScriptingFloatChannel, default=0, const [0,6]}
       CTRL_L_mouth_suckBlow {type=MovieSceneScriptingFloatChannel, default=0, const [0,6]}
       CTRL_L_mouth_pressD {type=MovieSceneScriptingFloatChannel, default=0, const [0,6]}
       CTRL_L_mouth_funnelU {type=MovieSceneScriptingFloatChannel, default=0, const [0,6]}
       CTRL_R_mouth_towardsU {type=MovieSceneScriptingFloatChannel, default=0, const [0,6]}
       CTRL_R_mouth_purseD {type=MovieSceneScriptingFloatChannel, default=0, const [0,6]}
       CTRL_L_mouth_lipBiteU {type=MovieSceneScriptingFloatChannel, default=0, const [0,6]}
       CTRL_R_mouth_pressD {type=MovieSceneScriptingFloatChannel, default=0, const [0,6]}
       CTRL_R_mouth_pressU {type=MovieSceneScriptingFloatChannel, default=0, const [0,6]}
       CTRL_L_mouth_purseD {type=MovieSceneScriptingFloatChannel, default=0, const [0,6]}
       CTRL_R_mouth_lipBiteU {type=MovieSceneScriptingFloatChannel, default=0, const [0,6]}
       CTRL_R_mouth_suckBlow {type=MovieSceneScriptingFloatChannel, default=0, thinned_from=6, keys [[0,1],[58,1],[198,0]]}
       CTRL_L_mouth_purseU {type=MovieSceneScriptingFloatChannel, default=0, const [0,6]}
       CTRL_R_mouth_towardsD {type=MovieSceneScriptingFloatChannel, default=0, const [0,6]}
       CTRL_L_mouth_lipBiteD {type=MovieSceneScriptingFloatChannel, default=0, const [0,6]}
       CTRL_R_mouth_purseU {type=MovieSceneScriptingFloatChannel, default=0, const [0,6]}
       CTRL_R_mouth_cornerDepress {type=MovieSceneScriptingFloatChannel, default=0, const [0,6]}
       CTRL_R_mouth_upperLipRaise {type=MovieSceneScriptingFloatChannel, default=0, const [0,6]}
       CTRL_R_mouth_dimple {type=MovieSceneScriptingFloatChannel, default=0, const [0,6]}
       CTRL_R_mouth_cornerPull {type=MovieSceneScriptingFloatChannel, default=0, const [0,6]}
      ]
     }]
   }
  ]
 }]
display_rate {denominator=1, numerator=30}
tick_resolution {denominator=1, numerator=24000}
tracks []
